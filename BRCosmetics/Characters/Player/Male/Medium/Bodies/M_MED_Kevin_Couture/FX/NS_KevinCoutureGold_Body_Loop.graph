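# Looping body cosmetic effect for the Kevin Couture Gold character, layering purple lightning, glows, runes and drifting gold flakes.
# Emitters follow the user-supplied skeletal mesh through random bone sampling and custom BoneFlow and partition modules that run sprites and ribbons along limbs.
# One-shot glow, decal and burst emitters accompany looping ones, with particle lights on lightning and feet.
# Hidden hand, foot and head flags mask particles via visibility tags or region masks.
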NiagaraSystem /BRCosmetics/Characters/Player/Male/Medium/Bodies/M_MED_Kevin_Couture/FX/NS_KevinCoutureGold_Body_Loop
dependencies {
 usedDataInterfaces [
  &NiagaraDataInterfaceArrayFloat
  &NiagaraDataInterfaceCurve
  &NiagaraDataInterfaceColorCurve
  &NiagaraDataInterfaceCamera
  &NiagaraDataInterfaceArrayInt32
 ]
 usedDynamicInputs [&MakeLinearColor, &Vector4FromCurve]
 usedModules [&SystemState]
}
emitters {
 AirGlow {
  emitterUpdateScript [
   &EmitterState(Life Cycle Mode=::NewEnumerator1, Loop Behavior=::NewEnumerator1, Loop Delay=0.5, Loop Duration=1, UseLoopDelay=false)
   &SpawnBurst_Instantaneous(Spawn Count=1, Use Spawn Probability=false)
  ]
  particleSpawnScript [
   &InitializeParticle(
    Color=<rgba 0.4159 0 1 0.1>
    Lifetime=0.65
    Sprite Size Mode=::NewEnumerator3
    Uniform Sprite Size=170
    UsePositionOffset=false
    Write Material Random=true
   ) {module=InitializeParticle001}
   &SystemLocation(MaskSpawnGroup=false, Offset=<0 0 115>)
  ]
  particleUpdateScript [
   &ParticleState()
   &ScaleColor(
    Scale Alpha=&FloatFromCurve(FloatCurve=DataInterface() {propertyValues {Curve {keys [
         {arriveTangent=5.867, interpMode=RCIM_Cubic, leaveTangent=5.867, tangentMode=RCTM_User, time=0, value=0}
         {arriveTangent=0.01, interpMode=RCIM_Cubic, leaveTangent=0.01, tangentMode=RCTM_User, time=0.25, value=1}
         {arriveTangent=-0.0518, interpMode=RCIM_Cubic, leaveTangent=-0.0518, tangentMode=RCTM_User, time=0.5, value=1}
         <at 1 0 RCIM_Cubic>
        ]}}})
    ScaleA=true
    ScaleRGB=true
   )
   &ScaleSpriteSize(Scale Factor=&Vector2DFromFloat(Value=&FloatFromCurve(FloatCurve=DataInterface() {propertyValues {Curve {keys [<at 0 0 RCIM_Cubic>,<at 0.2 1 RCIM_Cubic>]}}})))
   &DynamicMaterialParameters(
    Index 0 Param 1=2
    Module.Param0WriteEnabled=true
    Module.Param1WriteEnabled=true
    Module.Param2WriteEnabled=true
    Module.Param3WriteEnabled=true
   )
   &CameraOffset(Camera Offset Amount=-30)
  ]
  rendererClasses [&NiagaraSpriteRendererProperties]
  renderers [{rendererClass=&NiagaraSpriteRendererProperties}]
 }
 BoneFlow_Burst {
  emitterUpdateScript [
   &EmitterState(Life Cycle Mode=::NewEnumerator1, Loop Behavior=::NewEnumerator1, Loop Duration=0.75, UseLoopDelay=false)
   /Game/Effects/Niagara/Modules/Partition/FN_PartitionAttributes(Number of Bone Sets=5, Particles Per Partition=2, Partitions=5)
   &SpawnRate(SpawnRate=70, Use Spawn Probability=false)
  ]
  particleSpawnScript [
   /Game/Effects/Niagara/Modules/Partition/FN_Particle_partition_v5/FN_Particle_Partition_v5()
   /SetVariables(Particles.BoneflowPosition:NiagaraFloat=&Add_Float(A=$Particles.Partition.NormalizedID, B=$Emitter.NormalizedLoopAge)) {isSetParameters=true}
   &InitializeParticle(
    Color=<rgba 2.4952 0 6 1>
    Sprite Rotation Mode=::NewEnumerator2
    Sprite Size Mode=::NewEnumerator1
    Uniform Sprite Size Max=&SelectFloatFromArray(Direct Array Index=$Particles.Partition.ID, Float Selection Array=DataInterface() {propertyValues {FloatData [30,30,30,20,30]}})
    Uniform Sprite Size Min=&SelectFloatFromArray(Direct Array Index=$Particles.Partition.ID, Float Selection Array=DataInterface() {propertyValues {FloatData [20,20,20,10,20]}})
    UsePositionOffset=false
    Write Material Random=true
   )
   /Game/Effects/Niagara/Modules/BoneFlow/FN_Initialize_BoneFlow_Attributes_v2(
    Angle:NiagaraFloat=&UniformRangedFloat(Minimum=0, Maximum=1)
    Radius:NiagaraFloat=&UniformRangedFloat(Minimum=3, Maximum=10)
   )
   &DynamicMaterialParameters(
    Index 0 Param 3=&FloatFromCurve(FloatCurve=DataInterface() {propertyValues {Curve {keys [<at 0 1 RCIM_Cubic>,<at 0.2 0.025 RCIM_Cubic>,<at 1 1>]}}})
    Module.Param0WriteEnabled=false
    Module.Param1WriteEnabled=false
    Module.Param2WriteEnabled=true
    Module.Param3WriteEnabled=false
   ) {module=DynamicMaterialParameters001}
   /SetVariables(Particles.SubImageIndex=&UniformRangedFloat(Minimum=0, Maximum=3)) {isSetParameters=true, module=SetVariables_#1}
   /SetVariables(Particles.AirGlow_Color:LinearColor=<rgba 0.4159 0 1 1>, Particles.AirGlow_Size:Vector2f=<20 20>) {isSetParameters=true, module=SetVariables_#2}
   &CameraOffset(Camera Offset Amount=&MakeCustomFloatFromBool(
     Bool=&SetBoolByIntComparison(A=$Particles.Partition.ID, B=4, Comparison Type=::NewEnumerator4)
     True Float=4
     False Float=0
    ))
  ]
  particleUpdateScript [
   &ParticleState()
   &FN_BoneFlow(
    Angle:NiagaraFloat=$Particles.FN_Initialize_BoneFlow_Attributes_v2.Angle
    BoneID:NiagaraInt32=$Particles.Partition.ID
    Height:NiagaraFloat=$Particles.BoneflowPosition
    Interpolate Mesh Sampling=false
    Radius:NiagaraFloat=$Particles.FN_Initialize_BoneFlow_Attributes_v2.Radius
    Set Material Dynamic=false
    SkeletalMesh:NiagaraDataInterfaceSkeletalMesh=$User.SkeletalMesh
   )
   &DynamicMaterialParameters(
    Index 0 Param 3=&FloatFromCurve(FloatCurve=DataInterface() {propertyValues {Curve {keys [<at 0 1 RCIM_Cubic>,<at 0.2 0.025 RCIM_Cubic>,<at 1 1>]}}})
    Module.Param0WriteEnabled=false
    Module.Param1WriteEnabled=false
    Module.Param2WriteEnabled=true
    Module.Param3WriteEnabled=false
   )
   &ScaleColor(
    Scale Alpha=&FloatFromCurve(FloatCurve=DataInterface() {propertyValues {Curve {keys [<at 0.75 1 RCIM_Cubic>, {arriveTangent=-10.3917, interpMode=RCIM_Cubic, leaveTangent=-10.3917, tangentMode=RCTM_User, time=1, value=0}]}}})
    Scale RGB=&Multiply_VectorByFloat(
     Vector=&MakeVectorFromLinearColorRGB(Color=&ColorFromCurve(ColorCurve=DataInterface() {propertyValues {
         AlphaCurve {keys []}
         BlueCurve {keys [<at 0 1>,<at 0.15 10 RCIM_Cubic>,<at 1 1 RCIM_Cubic>]}
         GreenCurve {keys [<at 0 1>, {arriveTangent=-0.1823, interpMode=RCIM_Cubic, leaveTangent=-0.1823, time=0.15, value=9.045}, <at 1 0.8177 RCIM_Cubic>]}
         RedCurve {keys [<at 0 1>, {arriveTangent=-0.1812, interpMode=RCIM_Cubic, leaveTangent=-0.1812, time=0.15, value=8.5417}, <at 1 0.8188 RCIM_Cubic>]}
        }}))
     Float=1
    )
    ScaleA=true
    ScaleRGB=true
   )
   &SubUVAnimation(End Frame=8, SubUV Animation Mode:ENiagara_SubUVLookupMode=::NewEnumerator2) {module=SubUVAnimation001}
   &FN_Light_v3(
    Color=&ColorFromCurve(ColorCurve=DataInterface() {propertyValues {
       AlphaCurve {keys [<at 0 0 RCIM_Cubic>,<at 0.2 0.9664>,<at 0.4 0 RCIM_Cubic>]}
       BlueCurve {keys [<at 0 0 RCIM_Cubic>,<at 0.2 0.4>,<at 0.4 0 RCIM_Cubic>]}
       GreenCurve {keys [<at 0 0 RCIM_Cubic>,<at 0.2 0.0134>,<at 0.4 0 RCIM_Cubic>]}
       RedCurve {keys [<at 0 0 RCIM_Cubic>,<at 0.2 0.1742>,<at 0.4 0 RCIM_Cubic>]}
      }})
    Create Color Attribute=true
    Create Exponent Attribute=true
    Create Radius Attribute=true
    Radius=100
   )
   /SetVariables(Particles.VisibilityTag=&self:PartHiddenPartitionCheck(
     is Hand Hidden:NiagaraBool=$User.bisHandHidden
     Selection:NiagaraInt32=$Particles.Partition.ID
     is Foot Hidden:NiagaraBool=$User.bisFootHidden
    )) {isSetParameters=true}
  ]
  rendererClasses [&NiagaraSpriteRendererProperties, &NiagaraLightRendererProperties]
  renderers [
   {rendererClass=&NiagaraSpriteRendererProperties}
   {rendererClass=&NiagaraSpriteRendererProperties, rendererIndex=1}
   {rendererClass=&NiagaraLightRendererProperties, rendererIndex=2}
  ]
 }
 BoneFlow_Loop {
  emitterSpawnScript [&FN_PartitionAttributes(Number of Bone Sets=12, Particles Per Partition=5, Partitions=5)]
  emitterUpdateScript [
   &EmitterState(Inactive Response=::NewEnumerator1, Life Cycle Mode=::NewEnumerator1, Loop Duration=0.75, UseLoopDelay=false)
   &SpawnRate(SpawnRate=&UniformRangedFloat(Minimum=1, Maximum=10), Use Spawn Probability=false)
  ]
  particleSpawnScript [
   &FN_Particle_Partition_v5(Pattern:ENiagaraPartitionWaveForm=::NewEnumerator1, Shuffle Partition ID=true)
   &InitializeParticle(
    Color=<rgba 2.4952 0 6 1>
    Sprite Rotation Mode=::NewEnumerator2
    Sprite Size Mode=::NewEnumerator1
    Uniform Sprite Size Max=&SelectFloatFromArray(Direct Array Index=$Particles.Partition.ID, Float Selection Array=DataInterface() {propertyValues {FloatData [30,30,30,20,30]}})
    Uniform Sprite Size Min=&SelectFloatFromArray(Direct Array Index=$Particles.Partition.ID, Float Selection Array=DataInterface() {propertyValues {FloatData [20,20,20,10,20]}})
    UsePositionOffset=false
    Write Material Random=true
   ) {module=InitializeParticle001}
   &DynamicMaterialParameters(
    Index 0 Param 3=&FloatFromCurve(FloatCurve=DataInterface() {propertyValues {Curve {keys [<at 0 1 RCIM_Cubic>,<at 0.2 0.025 RCIM_Cubic>,<at 1 1>]}}})
    Module.Param0WriteEnabled=false
    Module.Param1WriteEnabled=false
    Module.Param2WriteEnabled=true
    Module.Param3WriteEnabled=false
   ) {module=DynamicMaterialParameters001}
   /SetVariables(Particles.SubImageIndex=&UniformRangedFloat(Minimum=0, Maximum=3)) {isSetParameters=true, module=SetVariables_#1}
   /SetVariables(Particles.AirGlow_Color:LinearColor=<rgba 0.4159 0 1 1>, Particles.AirGlow_Size:Vector2f=<20 20>) {isSetParameters=true}
   /SetVariables(Particles.RandomBone:NiagaraInt32=&RandomRangeInt(Minimum=0, Maximum=16)) {isSetParameters=true}
  ]
  particleUpdateScript [
   &ParticleState()
   &SkeletalMeshLocation(Bone / Socket Index=$Particles.RandomBone, Bone Sampling Mode=::NewEnumerator4, MaskSpawnGroup=false, Skeletal Mesh=$User.SkeletalMesh)
   &ShapeLocation(
    MaskSpawnGroup=false
    Override Seed=false
    Shape Origin=$Particles.SkeletalMeshLocation.SampledPosition
    Sphere Radius=&RandomRangeFloat(Minimum=1, Maximum=2)
    Sphere Surface Distribution=1
    UseOwnerScale=true
   )
   &DynamicMaterialParameters(
    Index 0 Param 3=&FloatFromCurve(FloatCurve=DataInterface() {propertyValues {Curve {keys [<at 0 1 RCIM_Cubic>,<at 0.2 0.025 RCIM_Cubic>,<at 1 1>]}}})
    Module.Param0WriteEnabled=false
    Module.Param1WriteEnabled=false
    Module.Param2WriteEnabled=true
    Module.Param3WriteEnabled=false
   )
   &ScaleColor(
    Scale Alpha=&FloatFromCurve(FloatCurve=DataInterface() {propertyValues {Curve {keys [<at 0.75 1 RCIM_Cubic>, {arriveTangent=-10.3917, interpMode=RCIM_Cubic, leaveTangent=-10.3917, tangentMode=RCTM_User, time=1, value=0}]}}})
    Scale RGB=&Multiply_VectorByFloat(
     Vector=&MakeVectorFromLinearColorRGB(Color=&ColorFromCurve(ColorCurve=DataInterface() {propertyValues {
         AlphaCurve {keys []}
         BlueCurve {keys [<at 0 1>,<at 0.15 10 RCIM_Cubic>,<at 1 1 RCIM_Cubic>]}
         GreenCurve {keys [<at 0 1>, {arriveTangent=-0.1823, interpMode=RCIM_Cubic, leaveTangent=-0.1823, time=0.15, value=9.045}, <at 1 0.8177 RCIM_Cubic>]}
         RedCurve {keys [<at 0 1>, {arriveTangent=-0.1812, interpMode=RCIM_Cubic, leaveTangent=-0.1812, time=0.15, value=8.5417}, <at 1 0.8188 RCIM_Cubic>]}
        }}))
     Float=1
    )
    ScaleA=true
    ScaleRGB=true
   )
   /Niagara/Modules/Update/SubUV/SubUVAnimation(Number Of Frames=8, Random SubUV Lookup Index:NiagaraFloat=$Particles.NormalizedAge, SubUV Animation Mode=::NewEnumerator1)
   &CameraOffset(Camera Offset Amount=4)
   &FN_Light_v3(
    Color=&ColorFromCurve(ColorCurve=DataInterface() {propertyValues {
       AlphaCurve {keys [<at 0 1 RCIM_Cubic>,<at 0.2 0.9664>,<at 0.4 0 RCIM_Cubic>]}
       BlueCurve {keys [<at 0 0 RCIM_Cubic>,<at 0.2 0.25>,<at 0.4 0 RCIM_Cubic>]}
       GreenCurve {keys [<at 0 0 RCIM_Cubic>,<at 0.2 0.0084>,<at 0.4 0 RCIM_Cubic>]}
       RedCurve {keys [<at 0 0 RCIM_Cubic>,<at 0.2 0.1089>,<at 0.4 0 RCIM_Cubic>]}
      }})
    Create Color Attribute=true
    Create Exponent Attribute=true
    Create Radius Attribute=true
    Radius=100
   )
   /SetVariables(Particles.VisibilityTag=&self:PartHiddenBoneCheck(
     bisHandHidden:NiagaraBool=$User.bisHandHidden
     bisFootHidden:NiagaraBool=$User.bisFootHidden
     biaHeadHidden:NiagaraBool=$User.bisHeadHidden
     Selection:NiagaraInt32=$Particles.RandomBone
    )) {isSetParameters=true}
  ]
  rendererClasses [&NiagaraSpriteRendererProperties, &NiagaraLightRendererProperties]
  renderers [
   {rendererClass=&NiagaraSpriteRendererProperties}
   {rendererClass=&NiagaraSpriteRendererProperties, rendererIndex=1}
   {rendererClass=&NiagaraLightRendererProperties, rendererIndex=2}
  ]
 }
 Bone_Flow_Ribbon02 {
  emitterUpdateScript [
   &EmitterState(
    Life Cycle Mode=::NewEnumerator1
    Loop Duration=&RandomRangeFloat(Minimum=0, Maximum=2)
    Recalculate Duration Each Loop=true
    UseLoopDelay=true
   )
   &SpawnBurst_Instantaneous(Spawn Count=6, Use Spawn Probability=false, UseLoopCountLimit=false)
   /SetVariables(
    Emitter.gen rand int:NiagaraInt32=&SelectIntFromArray(
     Array Sampling Mode=::NewEnumerator1
     Direct Array Index=?
     Evaluation Type=::NewEnumerator1
     Int Selection Array=DataInterface() {propertyValues {IntData [3,4,5,1,2,0,0,0,6,6,6]}}
     Random Seed=0
    )
    Emitter.genRand:NiagaraFloat=&RandomRangeFloat(Minimum=0, Maximum=1, Evaluation Type=::NewEnumerator1)
   ) {isSetParameters=true}
   &SpawnBurst_Instantaneous(Spawn Count=6, Spawn Time=0.6234, Use Spawn Probability=false, UseLoopCountLimit=false) {enabled=false, module=SpawnBurst_Instantaneous001}
  ]
  particleSpawnScript [
   &FN_SetSpawnGroup(Inclusivity=true, Normalize Spawn Group=true, Particles Per Group=6, Pattern=::NewEnumerator0, Set Ribbon ID and Ribbon Link Order=true)
   &InitializeRibbon(
    Color=<rgba 4.1587 0 10 1>
    Lifetime Max=0.3
    Lifetime Min=0.15
    Lifetime Mode:ENiagara_LifetimeMode=::NewEnumerator1
    Ribbon Width=55
    UsePositionOffset=false
    Write Material Random=true
   )
   &CameraOffset(Camera Offset Amount=10)
   &DynamicMaterialParameters(
    Index 0 Param 1=50
    Index 0 Param 3=&RandomRangeFloat(Minimum=0.3, Maximum=0.1)
    Index 0 Param 4=1
    Module.Param0WriteEnabled=true
    Module.Param1WriteEnabled=true
    Module.Param2WriteEnabled=true
    Module.Param3WriteEnabled=true
   ) {enabled=false}
   /SetVariables(Particles.random:NiagaraFloat=$Emitter.genRand, Particles.randomInt:NiagaraInt32=$Emitter.gen rand int) {isSetParameters=true}
  ]
  particleUpdateScript [
   &ParticleState()
   &FN_BoneFlow(
    Angle:NiagaraFloat=&Add_Float(
     A=&SelectFloatFromArray(Direct Array Index=$Particles.Initial.randomInt, Float Selection Array=DataInterface() {propertyValues {FloatData [0.2,0.5,0.6]}})
     B=&RandomRangeFloat(Minimum=-0.05, Maximum=0.05)
    )
    BoneID:NiagaraInt32=$Particles.Initial.randomInt
    Enable Position Offsets=true
    Height:NiagaraFloat=$Particles.Normalized Spawn Group
    Offset Mid:Vector3f=&Lerp_Vector(
     A=<0 0 0>
     B=&RandomRangeVector(
      Minimum=<0 0 0>
      Maximum=&RandomVector(Vector Scale=&RandomRangeFloat(Minimum=0, Maximum=0))
     )
     Alpha=$Particles.NormalizedAge
    )
    Radius:NiagaraFloat=&Multiply_Float(
     A=&FloatFromCurve(
      FloatCurve=DataInterface() {propertyValues {Curve {keys [
          {arriveTangent=4.3816, interpMode=RCIM_Cubic, leaveTangent=4.3816, tangentMode=RCTM_User, time=0, value=0}
          {interpMode=RCIM_Cubic, tangentMode=RCTM_User, time=0.5, value=1}
          {arriveTangent=-4.6838, interpMode=RCIM_Cubic, leaveTangent=-4.6838, tangentMode=RCTM_User, time=1, value=0}
         ]}}}
      CurveIndex=$Particles.Normalized Spawn Group
      Scale Curve=&RandomRangeFloat(Minimum=0, Maximum=15)
     )
     B=&RandomRangeFloat(Minimum=1, Maximum=1)
    )
    Set Material Dynamic=false
    SkeletalMesh:NiagaraDataInterfaceSkeletalMesh=$User.SkeletalMesh
    Use Sockets Instead=true
   )
   /SetVariables(Particles.RibbonFacing=$Output.FN_BoneFlow.SplineNormal) {isSetParameters=true}
   &ScaleRibbonWidth(Ribbon Width Scale=&Multiply_Float(
     A=&Multiply_Float(
      A=&FloatFromCurve(
       FloatCurve=DataInterface() {propertyValues {Curve {keys [
           {interpMode=RCIM_Cubic, tangentMode=RCTM_User, time=0, value=0}
           <at 0.1 0>
           {interpMode=RCIM_Cubic, tangentMode=RCTM_Break, time=0.2, value=1}
           {arriveTangent=-1.5728, interpMode=RCIM_Cubic, leaveTangent=-1.5728, tangentMode=RCTM_User, time=1, value=0.5}
          ]}}}
       Scale Curve=&RandomRangeFloat(Minimum=0.7, Maximum=1)
      )
      B=0.75
     )
     B=&FloatFromCurve(
      FloatCurve=DataInterface() {propertyValues {Curve {keys [
          {interpMode=RCIM_Cubic, tangentMode=RCTM_User, time=0, value=0.675}
          {interpMode=RCIM_Cubic, tangentMode=RCTM_User, time=0.5, value=0.5}
          {tangentMode=RCTM_User, time=1, value=0.675}
         ]}}}
      CurveIndex=$Particles.Normalized Spawn Group
      Scale Curve=&FloatFromCurve(
       FloatCurve=DataInterface() {propertyValues {Curve {keys [
           {interpMode=RCIM_Cubic, tangentMode=RCTM_User, time=0, value=1}
           {interpMode=RCIM_Cubic, tangentMode=RCTM_User, time=0.3, value=0.4}
           {interpMode=RCIM_Cubic, tangentMode=RCTM_User, time=0.5, value=1.5}
           {interpMode=RCIM_Cubic, tangentMode=RCTM_User, time=0.7, value=0.4}
           {interpMode=RCIM_Cubic, tangentMode=RCTM_User, time=1, value=1}
          ]}}}
       CurveIndex=&Multiply_Float(
        A=$Particles.Normalized Spawn Group
        B=&Multiply_Float(A=$Particles.Age, B=&RandomRangeFloat(Minimum=1, Maximum=1.5))
       )
      )
     )
    ))
   &ScaleColor(
    Scale Alpha=&FloatFromCurve(FloatCurve=DataInterface() {propertyValues {Curve {keys [
         {arriveTangent=8.0553, interpMode=RCIM_Cubic, leaveTangent=8.0553, tangentMode=RCTM_User, time=0, value=0}
         <at 0.2113 1 RCIM_Cubic>
         <at 0.4546 0>
         {arriveTangent=-8.0018, interpMode=RCIM_Cubic, leaveTangent=7.1228, tangentMode=RCTM_Break, time=0.5517, value=0}
         <at 0.7524 1 RCIM_Cubic>
         {arriveTangent=-10.4882, interpMode=RCIM_Cubic, leaveTangent=-10.4882, tangentMode=RCTM_User, time=1, value=0}
        ]}}})
    ScaleA=true
    ScaleRGB=true
   ) {module=ScaleColor001}
   &SolveForcesAndVelocity(Clamp Velocity=false, Limit Acceleration=false)
   &RegionMask(
    Invert Region=true
    Offset:Vector3f=&MakeCustomVectorFromBool(Bool=$User.bisFootHidden, True Vector=<0 0 50>, False Vector=<0 0 0>)
    Offset Space:ENiagaraCoordinateSpace=::Local
    Region Type:ENiagaraDebugDrawShape=::NewEnumerator2
    Rotation:Vector3f=<0 0.25 0>
    UseOffset=true
    UseRotation=false
   )
   &ScaleColor(Scale Alpha=$Particles.RegionMask.Float, ScaleA=true, ScaleRGB=true)
  ]
  rendererClasses [&NiagaraRibbonRendererProperties]
  renderers [{rendererClass=&NiagaraRibbonRendererProperties}]
 }
 BurstRunes {
  emitterUpdateScript [
   &EmitterState()
   &SpawnRate(SpawnRate=&UniformRangedFloat(Minimum=0, Maximum=10), Use Spawn Probability=false)
  ]
  particleSpawnScript [
   &InitializeParticle(
    Color Channel Mode:ENiagaraChannelCorrelation=::NewEnumerator2
    Color Maximum=<rgba 1 10 10 1>
    Color Minimum=<rgba 1 5 1 0.5>
    Color Mode=::NewEnumerator2
    Lifetime=&UniformRangedFloat(Minimum=0.3, Maximum=&UniformAOrBFloat(A=0.6, B=1.1, Distribution Weight=0.25))
    Sprite Rotation Mode=::NewEnumerator2
    Sprite Size Mode=::NewEnumerator1
    UsePositionOffset=false
    Write Material Random=true
   ) {module=InitializeParticle001}
   /SetVariables(Particles.SpriteInitRotationRate:NiagaraFloat=&Multiply_Float(
     A=&UniformAOrBFloat(A=-1)
     B=&UniformRangedFloat(Minimum=5, Maximum=10)
    )) {isSetParameters=true}
   /SetVariables(Particles.ZeroOrOne:NiagaraFloat=&UniformAOrBFloat(A=0, Distribution Weight=0.25)) {isSetParameters=true}
   /SetVariables(
    Particles.RayColor:LinearColor=&RandomRangeLinearColor(Minimum=<rgba 0.5191 0.3646 5 0.75>, Maximum=<rgba 20 0 7.1984 1>)
    Particles.RaySize:Vector2f=&Lerp_Vector2(A=<0 0>, B=&UniformRangedVector2D(Minimum=<15 10>, Maximum=<25 35>), Alpha=$Particles.ZeroOrOne)
    Particles.RayVelocity:Vector3f=&Multiply_Vector(A=$Particles.Velocity, B=<0.25 1 0.25>)
   ) {isSetParameters=true}
   /Niagara/Modules/Update/SubUV/V2/SubUVAnimation(End Frame=4, SubUV Animation Mode:ENiagara_SubUVLookupMode=::NewEnumerator2) {module=SubUVAnimation001}
   /SetVariables(Particles.RayPos:Vector3f=&Add_Vector(A=Linked() {linkedVariable {name=Particles.Position, type=NiagaraPosition}}, B=<0 5 0>)) {isSetParameters=true}
   /SetVariables(Particles.RandomBone:NiagaraInt32=&UniformRangedInt(Minimum=0, Maximum=17)) {isSetParameters=true}
   /SetVariables(Particles.Position=&UniformRangedVector(Minimum=<-7 -7 -7>, Maximum=<5 5 5>)) {isSetParameters=true}
  ]
  particleUpdateScript [
   &ParticleState()
   &SkeletalMeshLocation(
    Bone / Socket Index=$Particles.RandomBone
    Bone Sampling Mode=::NewEnumerator4
    Inherit Velocity=false
    MaskSpawnGroup=false
    Sampled Position Offset:NiagaraPosition=$Particles.Initial.Position
    Sampled Position Offset Coordinate Space:ENiagaraMeshTransforms=::NewEnumerator1
    Skeletal Mesh=$User.SkeletalMesh
   )
   &SpriteRotationRate(Rotation Rate=&Multiply_Float(A=$Particles.SpriteInitRotationRate, B=&VectorLength(Vector=$Particles.Velocity)))
   &ScaleColor(
    Scale Alpha=&FloatFromCurve(FloatCurve=DataInterface() {propertyValues {Curve {keys [{interpMode=RCIM_Cubic, tangentMode=RCTM_User, time=0.75, value=1}, <at 1 0.5>]}}})
    Scale RGB=&MakeVector(
     X=&UniformAOrBFloat(A=0)
     Y=1
     Z=&FloatFromCurve(FloatCurve=DataInterface() {propertyValues {Curve {keys [
          {arriveTangent=-32.9921, interpMode=RCIM_Cubic, leaveTangent=-32.9921, tangentMode=RCTM_User, time=0.2, value=10}
          {arriveTangent=-0.0507, interpMode=RCIM_Cubic, leaveTangent=-0.0508, tangentMode=RCTM_User, time=0.3, value=1}
          {arriveTangent=8.5796, interpMode=RCIM_Cubic, leaveTangent=8.5796, tangentMode=RCTM_User, time=1, value=20}
         ]}}})
    )
    ScaleA=true
    ScaleRGB=true
   )
   /SetVariables(Particles.RayColor:LinearColor=&MultiplyLinearColor(
     Linear Color=$Particles.Initial.RayColor
     Scale Factor=&MakeFloat4FromVectorAndFloat(
      "Vector (RGB)"=&Lerp_Vector(A=<1 1 1>, B=<0.1 0.1 5>, Alpha=$Particles.NormalizedAge)
      "Float (Alpha)"=&ClampFloat(Float=&Lerp_Float(A=1, B=-0.25, Alpha=$Particles.NormalizedAge))
     )
    )) {isSetParameters=true}
   &CameraOffset(Camera Offset Amount=40)
   &ScaleSpriteSize(
    Scale Sprite Size Mode=::NewEnumerator0
    Uniform Scale Factor=&Lerp_Float(
     A=0
     B=1
     Alpha=&NormalizeDistanceRange(
      Start Position:NiagaraPosition=$Particles.Position
      End Position:NiagaraPosition=&ReturnCameraVector(Camera Property=::NewEnumerator0, Coordinate Space=::Simulation)
      Distance:NiagaraFloat=200
      Invert Normalized Range:NiagaraBool=false
     )
    )
   )
   /SetVariables(Particles.VisibilityTag=&self:PartHiddenBoneCheck(
     bisHandHidden:NiagaraBool=$User.bisHandHidden
     bisFootHidden:NiagaraBool=$User.bisFootHidden
     biaHeadHidden:NiagaraBool=$User.bisHeadHidden
     Selection:NiagaraInt32=$Particles.RandomBone
    )) {isSetParameters=true}
  ]
  rendererClasses [&NiagaraSpriteRendererProperties]
  renderers [
   {rendererClass=&NiagaraSpriteRendererProperties}
   {rendererClass=&NiagaraSpriteRendererProperties, rendererIndex=1}
  ]
 }
 Decal {
  emitterUpdateScript [
   &EmitterState(Life Cycle Mode=::NewEnumerator1, Loop Behavior=::NewEnumerator1, Loop Duration=1, UseLoopDelay=true)
   &SpawnBurst_Instantaneous(Spawn Count=12, Use Spawn Probability=false)
  ]
  particleSpawnScript [
   &FN_SetSpawnGroup_v2(Number of Spawn Groups=12) {enabled=false}
   &FN_SetSpawnGroup(Number of Spawn Groups=12)
   &InitializeParticle(
    Color=<rgba 0.4159 0 1 1>
    Lifetime=0.65
    Sprite Size Mode=::NewEnumerator3
    Uniform Sprite Size=50
    UsePositionOffset=false
    Write Material Random=true
   ) {module=InitializeParticle001}
   &DynamicMaterialParameters(
    Index 0 Param 1=0.5
    Index 0 Param 2=1
    Index 0 Param 3=1
    Module.Param0WriteEnabled=false
    Module.Param1WriteEnabled=true
    Module.Param2WriteEnabled=true
    Module.Param3WriteEnabled=true
   )
  ]
  particleUpdateScript [
   &SkeletalMeshLocation(
    Apply Sampled Orientation to Mesh Particles=true
    Bone / Socket Index=&SpawnGroup()
    Bone Sampling Mode=::NewEnumerator4
    Inherit Velocity=false
    MaskSpawnGroup=false
    Sampled Position Offset=&Lerp_Vector(A=<0 0 -75>, B=<0 0 300>, Alpha=$Particles.NormalizedAge)
    Skeletal Mesh=$User.SkeletalMesh
   )
   &ParticleState()
   &DynamicMaterialParameters(
    Index 0 Param 1=&FloatFromCurve(FloatCurve=DataInterface() {propertyValues {Curve {keys [<at 0.4 2>,<at 1 6>]}}})
    Module.Param0WriteEnabled=true
    Module.Param1WriteEnabled=false
    Module.Param2WriteEnabled=false
    Module.Param3WriteEnabled=false
   ) {module=DynamicMaterialParameters001}
   &ScaleColor(
    Scale Alpha=&FloatFromCurve(FloatCurve=DataInterface() {propertyValues {Curve {keys [<at 0 10 RCIM_Cubic>, {arriveTangent=-10, interpMode=RCIM_Cubic, leaveTangent=-10, time=0.2, value=1}, <at 1 0 RCIM_Cubic>]}}})
    ScaleA=true
    ScaleRGB=true
   )
   &CameraOffset(Camera Offset Amount=20)
   /SetVariables(Particles.VisibilityTag=&self:PartHiddenBoneCheck(
     bisHandHidden:NiagaraBool=$User.bisHandHidden
     bisFootHidden:NiagaraBool=$User.bisFootHidden
     biaHeadHidden:NiagaraBool=$User.bisHeadHidden
     Selection:NiagaraInt32=$Particles.SpawnGroup
    )) {isSetParameters=true, module=SetVariables_#3}
  ]
  rendererClasses [&NiagaraSpriteRendererProperties]
  renderers [{rendererClass=&NiagaraSpriteRendererProperties}]
 }
 FeetLights {
  emitterUpdateScript [
   &EmitterState(
    Inactive Response=::NewEnumerator1
    Life Cycle Mode=::NewEnumerator1
    Loop Behavior=::NewEnumerator1
    Loop Duration Mode=::NewEnumerator1
    UseLoopDelay=false
   )
   &SpawnBurst_Instantaneous(Spawn Count=2, Use Spawn Probability=false)
  ]
  particleSpawnScript [
   &InitializeParticle(Lifetime=5, Sprite Size Mode=::NewEnumerator3, Uniform Sprite Size=40, UsePositionOffset=false, Write Material Random=true) {module=InitializeParticle001}
   &FN_SetSpawnGroup_v2(Number of Spawn Groups=2) {enabled=false}
   &FN_SetSpawnGroup(Number of Spawn Groups=2)
   /SetVariables(
    Particles.AirGlow_Color:LinearColor=&MakeCustomLinearColorFromBool(Bool=$User.bisFootHidden, True Color=<rgba 1 0.6667 0 0>, False Color=<rgba 1 0.6667 0 0.5>)
    Particles.AirGlow_Size:Vector2f=<60 60>
   ) {isSetParameters=true}
  ]
  particleUpdateScript [
   &ParticleState(Kill Particles When Lifetime Has Elapsed=false)
   &SkeletalMeshLocation(
    Bone / Socket Index=&Add_Integer(A=&SpawnGroup(), B=15)
    Bone Sampling Mode=::NewEnumerator4
    Inherit Velocity=false
    MaskSpawnGroup=false
    Sampled Position Offset=<0 0 -5>
    Skeletal Mesh=$User.SkeletalMesh
   )
   &FN_Light_v3(
    Color=<rgba 3 3 0 1>
    Create Color Attribute=true
    Create Exponent Attribute=true
    Create Position Attribute=true
    Create Radius Attribute=true
    Position:Vector3f=&Add_Vector(A=Linked() {linkedVariable {name=Particles.Position, type=NiagaraPosition}}, B=<0 10 10>)
    Radius=50
   )
   /SetVariables(Particles.AirGlow_Color:LinearColor=&MakeCustomLinearColorFromBool(Bool=$User.bisFootHidden, True Color=<rgba 1 0.6667 0 0>, False Color=<rgba 1 0.6667 0 0.5>)) {isSetParameters=true}
   /SetVariables(Particles.VisibilityTag=&MakeCustomIntFromBool(Bool=$User.bisFootHidden, True INT=1, False INT=0)) {isSetParameters=true}
  ]
  rendererClasses [&NiagaraSpriteRendererProperties, &NiagaraLightRendererProperties]
  renderers [
   {rendererClass=&NiagaraSpriteRendererProperties}
   {rendererClass=&NiagaraLightRendererProperties, rendererIndex=1}
  ]
 }
 Gold_Flakes001 {
  emitterUpdateScript [
   &EmitterState(Life Cycle Mode=::NewEnumerator1, Loop Duration=2, UseLoopDelay=false)
   &SpawnRate(SpawnRate=0, Use Spawn Probability=false) {enabled=false}
  ]
  particleSpawnScript [
   &InitializeParticle(
    Color=<rgba 0.6 0.3096 0.0062 1>
    Lifetime=1.25
    Mass Mode=::NewEnumerator3
    Mesh Scale Mode=::NewEnumerator1
    Mesh Uniform Scale Max=0.01
    Mesh Uniform Scale Min=0.015
    Sprite Size Mode=::NewEnumerator1
    Uniform Sprite Size Max=5
    Uniform Sprite Size Min=3
    UsePositionOffset=false
    Write Material Random=true
   )
   &DynamicMaterialParameters(
    Module.Param0WriteEnabled=true
    Module.Param0WriteEnabled1=true
    Module.Param1WriteEnabled=true
    Module.Param1WriteEnabled1=true
    Module.Param2WriteEnabled=true
    Module.Param2WriteEnabled1=true
    Module.Param3WriteEnabled=true
    Module.Param3WriteEnabled1=true
    Write Parameter Index 1=true
   )
   &AddVelocity()
   &SkeletalMeshLocation(
    Bone Sampling Mode=::NewEnumerator4
    MaskSpawnGroup=false
    Orientation Sampling=::NewEnumerator1
    Position Sampling=::NewEnumerator2
    Sampled Position Offset=<5 5 0>
    Skeletal Mesh=$User.SkeletalMesh
   )
   &AddVelocityFromPoint(Module.Use Falloff Distance=false, Origin Offset=<-0.1 0 0>, Velocity Strength=6)
   &InitialMeshOrientation(Use Rotation Vector=false)
  ]
  particleUpdateScript [
   &ParticleState()
   &InheritSourceMovement(
    Applied Position Scale=&Multiply_VectorByFloat(
     Vector=<1 1 1>
     Float=&FloatFromCurve(FloatCurve=DataInterface() {propertyValues {Curve {keys [<at 0 1>,<at 1 0.8>]}}})
    )
    ApplyForce=false
    ApplyPosition=true
    ApplyPositionFalloff=false
   )
   &CurlNoiseForce(Noise Frequency=75, Noise Strength=100, Pan Noise=true, Pan Noise Field=<0 0 1>)
   &GravityForce(Gravity=<0 0 100>)
   &Drag(Drag=3, Use Linear Drag=true, Use Rotational Drag=false)
   &SolveForcesAndVelocity(Clamp Velocity=false, Limit Acceleration=false)
   &ScaleSpriteSize(Uniform Curve Sprite Scale=DataInterface() {propertyValues {Curve {keys [
        {interpMode=RCIM_Cubic, tangentMode=RCTM_User, time=0, value=0}
        {interpMode=RCIM_Cubic, tangentMode=RCTM_User, time=0.3, value=1}
        {arriveTangent=-5.5933, interpMode=RCIM_Cubic, leaveTangent=-5.5933, tangentMode=RCTM_User, time=1, value=0}
       ]}}})
   &SpriteFacingAndAlignment(
    Sprite Facing=&Lerp_Vector(
     A=&RandomVector(Vector Scale=1)
     B=&RandomVector(Vector Scale=1)
     Alpha=$Particles.NormalizedAge
    )
    WriteAlignment=false
    WriteFacing=true
   )
   &UpdateMeshOrientation(
    Rotation Rate=&RandomRangeFloat(Minimum=0.5, Maximum=1.5)
    Rotation Vector=&RandomVector(Vector Scale=1)
   )
   &ScaleMeshSize(Scale Factor=&Multiply_Vector(
     A=&VectorFromFloat(Value=&FloatFromCurve(FloatCurve=DataInterface() {propertyValues {Curve {keys [
           <at 0 0 RCIM_Cubic>
           <at 0.3 1 RCIM_Cubic>
           {arriveTangent=-4.89, interpMode=RCIM_Cubic, leaveTangent=-4.89, tangentMode=RCTM_User, time=1, value=0}
          ]}}}))
     B=&VectorFromFloat(Value=&Lerp_Float(
       A=-0.1
       B=1
       Alpha=&NormalizeDistanceRange(
        Start Position:NiagaraPosition=$Particles.Position
        End Position:NiagaraPosition=&ReturnCameraVector(Camera Property=::NewEnumerator0, Coordinate Space=::Simulation)
        Distance:NiagaraFloat=300
        Invert Normalized Range:NiagaraBool=false
       )
      ))
    )) {module=ScaleMeshSize001}
   /SetVariables(Particles.VisibilityTag=&MakeCustomIntFromBool(Bool=$User.bisHeadHidden, True INT=1, False INT=0)) {isSetParameters=true}
  ]
  rendererClasses [&NiagaraSpriteRendererProperties, &NiagaraMeshRendererProperties]
  renderers [
   {rendererClass=&NiagaraSpriteRendererProperties}
   {rendererClass=&NiagaraMeshRendererProperties, rendererIndex=1}
  ]
 }
 Gold_Flakes002 {
  emitterUpdateScript [
   &EmitterState(Life Cycle Mode=::NewEnumerator1, Loop Duration=2, UseLoopDelay=false)
   &SpawnRate(SpawnRate=30, Use Spawn Probability=false) {module=SpawnRate001}
  ]
  particleSpawnScript [
   /SetVariables(Particles.RandomBone:NiagaraInt32=&UniformRangedInt(Minimum=0, Maximum=21)) {isSetParameters=true}
   &InitializeParticle(
    Color=<rgba 0.8 0 0 1>
    Lifetime Max=1
    Lifetime Min=2
    Lifetime Mode=::NewEnumerator1
    Mass Mode=::NewEnumerator3
    Mesh Scale Mode=::NewEnumerator1
    Mesh Uniform Scale Max=0.015
    Mesh Uniform Scale Min=0.01
    Sprite Size Mode=::NewEnumerator1
    Uniform Sprite Size Max=3.5
    Uniform Sprite Size Min=2
    UsePositionOffset=false
    Write Material Random=true
   )
   &SkeletalMeshLocation(Bone / Socket Index=$Particles.RandomBone, Bone Sampling Mode=::NewEnumerator4, MaskSpawnGroup=false, Skeletal Mesh=$User.SkeletalMesh)
   &SphereLocation(
    MaskSpawnGroup=false
    Override Seed=false
    Sphere Radius=6
    Surface Only=false
    UseNonUniformScale=false
    UseOffset=false
    UseSphereOrientation=false
   )
   &InheritVelocity(
    Inherited Velocity Amount Scale=<-0.1 -0.1 -0.5>
    Speed Limit=true
    UseSpeedThreshold=false
    Velocity Source=$Particles.SkeletalMeshLocation.SampledVelocity
   )
   &InitialMeshOrientation(Use Rotation Vector=false)
   &KillParticles(Kill Particles=&self:Wave_Data(Divisor:NiagaraFloat=-8, Invert:NiagaraBool=true, Offset:Vector3f=<-600 -300 500>, threashold:NiagaraFloat=0, time offset:NiagaraFloat=2)) {module=KillParticles002}
   &KillParticles() {module=KillParticles003}
   /SetVariables() {isSetParameters=true}
  ]
  particleUpdateScript [
   &ParticleState()
   &Drag(Drag=2.5, Use Linear Drag=true, Use Rotational Drag=false)
   &Color(Color=&ColorFromCurve(ColorCurve=DataInterface() {propertyValues {
       AlphaCurve {keys [<at 0 1 RCIM_Cubic>]}
       BlueCurve {keys [<at 0 2.5>,<at 0.188 0.0062 RCIM_Cubic>]}
       GreenCurve {keys [<at 0 17.5>,<at 0.188 0.3096 RCIM_Cubic>]}
       RedCurve {keys [<at 0 25>,<at 0.188 0.6 RCIM_Cubic>]}
      }}))
   &InheritSourceMovement(
    Applied Position Falloff Distance=150
    ApplyForce=false
    ApplyPosition=true
    ApplyPositionFalloff=false
    Source Velocity:Vector3f=&Add_Vector(
     A=$Engine.Owner.Velocity
     B=&Multiply_VectorByFloat(Vector=$StackContext.SkeletalMeshLocation.SampledVelocity, Float=-0.1)
    )
   )
   &CurlNoiseForce(Noise Frequency=75, Noise Strength=50, Pan Noise=true, Pan Noise Field=<0 0 1>)
   &GravityForce(Gravity=<0 0 45>) {module=GravityForce001}
   &GravityForce(Coordinate Space=::Local, Gravity=<16 -16 0>)
   &SolveForcesAndVelocity(Clamp Velocity=false, Limit Acceleration=false)
   &ScaleSpriteSize(Uniform Curve Sprite Scale=DataInterface() {propertyValues {Curve {keys [
        <at 0 0 RCIM_Cubic>
        <at 0.3 1 RCIM_Cubic>
        {arriveTangent=-4.89, interpMode=RCIM_Cubic, leaveTangent=-4.89, tangentMode=RCTM_User, time=1, value=0}
       ]}}})
   &SpriteFacingAndAlignment(
    Sprite Facing=&Lerp_Vector(
     A=&RandomVector(Vector Scale=1)
     B=&RandomVector(Vector Scale=1)
     Alpha=$Particles.NormalizedAge
    )
    WriteAlignment=false
    WriteFacing=true
   )
   &UpdateMeshOrientation(
    Rotation Rate=&RandomRangeFloat(Minimum=0.5, Maximum=1.25)
    Rotation Vector=&RandomVector(Vector Scale=1)
   )
   &ScaleMeshSize(Scale Factor=&Multiply_Vector(
     A=&VectorFromFloat(Value=&FloatFromCurve(FloatCurve=DataInterface() {propertyValues {Curve {keys [
           <at 0 0 RCIM_Cubic>
           <at 0.3 1 RCIM_Cubic>
           {arriveTangent=-4.89, interpMode=RCIM_Cubic, leaveTangent=-4.89, tangentMode=RCTM_User, time=1, value=0}
          ]}}}))
     B=&VectorFromFloat(Value=&Lerp_Float(
       A=0
       B=1
       Alpha=&NormalizeDistanceRange(
        Start Position:NiagaraPosition=$Particles.Position
        End Position:NiagaraPosition=&ReturnCameraVector(Camera Property=::NewEnumerator0, Coordinate Space=::Simulation)
        Distance:NiagaraFloat=300
        Invert Normalized Range:NiagaraBool=false
       )
      ))
    )) {module=ScaleMeshSize001}
   /SetVariables(Particles.VisibilityTag=&self:PartHiddenBoneCheck(
     bisHandHidden:NiagaraBool=$User.bisHandHidden
     bisFootHidden:NiagaraBool=$User.bisFootHidden
     biaHeadHidden:NiagaraBool=$User.bisHeadHidden
     Selection:NiagaraInt32=$Particles.RandomBone
    )) {isSetParameters=true}
  ]
  rendererClasses [&NiagaraSpriteRendererProperties, &NiagaraMeshRendererProperties]
  renderers [
   {rendererClass=&NiagaraSpriteRendererProperties}
   {rendererClass=&NiagaraMeshRendererProperties, rendererIndex=1}
  ]
 }
 InitialBurst_Lightning {
  emitterUpdateScript [
   &EmitterState(Life Cycle Mode=::NewEnumerator1, Loop Behavior=::NewEnumerator1, Loop Duration=0.75, UseLoopDelay=false)
   &FN_PartitionAttributes(Number of Bone Sets=5, Partitions=5)
   &SpawnBurst_Instantaneous(Spawn Count=20, Use Spawn Probability=false)
  ]
  particleSpawnScript [
   /Game/Effects/Niagara/Modules/Partition/FN_ParticlePartition_v2/FN_Particle_Partition() {enabled=false}
   &FN_Particle_Partition_v5()
   &FN_SetSpawnGroup_v2(Number of Spawn Groups=5) {enabled=false}
   &FN_SetSpawnGroup(Number of Spawn Groups=5)
   &InitializeParticle(
    Color=<rgba 2.4952 0 6 1>
    Sprite Rotation Mode=::NewEnumerator2
    Sprite Size Mode=::NewEnumerator3
    Uniform Sprite Size=&MaskFloatBySpawnGroup(
     Masked Float=&UniformRangedFloat(Minimum=20, Maximum=30)
     Passthrough Float=&UniformRangedFloat(Minimum=10, Maximum=20)
     Spawn Group Mask=4
    )
    UsePositionOffset=false
    Write Material Random=true
   ) {module=InitializeParticle001}
   &DynamicMaterialParameters(
    Index 0 Param 3=&FloatFromCurve(FloatCurve=DataInterface() {propertyValues {Curve {keys [<at 0 1 RCIM_Cubic>,<at 0.2 0.025 RCIM_Cubic>,<at 1 1>]}}})
    Module.Param0WriteEnabled=false
    Module.Param1WriteEnabled=false
    Module.Param2WriteEnabled=true
    Module.Param3WriteEnabled=false
   ) {module=DynamicMaterialParameters001}
   /SetVariables(Particles.SubImageIndex=&UniformRangedFloat(Minimum=0, Maximum=3)) {isSetParameters=true, module=SetVariables_#1}
   /SetVariables(Particles.AirGlow_Color:LinearColor=<rgba 0.4159 0 1 1>, Particles.AirGlow_Size:Vector2f=<20 20>) {isSetParameters=true, module=SetVariables_#2}
   /SetVariables(Particles.RandomBone:NiagaraInt32=&RandomRangeInt(Minimum=0, Maximum=21)) {isSetParameters=true}
  ]
  particleUpdateScript [
   &ParticleState()
   &SkeletalMeshLocation(
    Bone / Socket Index=$Particles.RandomBone
    Bone Sampling Mode=::NewEnumerator4
    Inherit Velocity=false
    MaskSpawnGroup=false
    Skeletal Mesh=$User.SkeletalMesh
   )
   &DynamicMaterialParameters(
    Index 0 Param 3=&FloatFromCurve(FloatCurve=DataInterface() {propertyValues {Curve {keys [<at 0 1 RCIM_Cubic>,<at 0.2 0.025 RCIM_Cubic>,<at 1 1>]}}})
    Module.Param0WriteEnabled=false
    Module.Param1WriteEnabled=false
    Module.Param2WriteEnabled=true
    Module.Param3WriteEnabled=false
   )
   &ScaleColor(
    Scale Alpha=&FloatFromCurve(FloatCurve=DataInterface() {propertyValues {Curve {keys [<at 0.75 1 RCIM_Cubic>, {arriveTangent=-10.3917, interpMode=RCIM_Cubic, leaveTangent=-10.3917, tangentMode=RCTM_User, time=1, value=0}]}}})
    Scale RGB=&Multiply_VectorByFloat(
     Vector=&MakeVectorFromLinearColorRGB(Color=&ColorFromCurve(ColorCurve=DataInterface() {propertyValues {
         AlphaCurve {keys []}
         BlueCurve {keys [<at 0 1>,<at 0.15 10 RCIM_Cubic>,<at 1 1 RCIM_Cubic>]}
         GreenCurve {keys [<at 0 1>, {arriveTangent=-0.1823, interpMode=RCIM_Cubic, leaveTangent=-0.1823, time=0.15, value=9.045}, <at 1 0.8177 RCIM_Cubic>]}
         RedCurve {keys [<at 0 1>, {arriveTangent=-0.1812, interpMode=RCIM_Cubic, leaveTangent=-0.1812, time=0.15, value=8.5417}, <at 1 0.8188 RCIM_Cubic>]}
        }}))
     Float=1
    )
    ScaleA=true
    ScaleRGB=true
   )
   &SubUVAnimation(End Frame=8, SubUV Animation Mode:ENiagara_SubUVLookupMode=::NewEnumerator2) {module=SubUVAnimation001}
   &FN_Light_v3(
    Color=&ColorFromCurve(ColorCurve=DataInterface() {propertyValues {
       AlphaCurve {keys [<at 0 0 RCIM_Cubic>,<at 0.2 0.9664>,<at 0.4 0 RCIM_Cubic>]}
       BlueCurve {keys [<at 0 0 RCIM_Cubic>,<at 0.2 0.4>,<at 0.4 0 RCIM_Cubic>]}
       GreenCurve {keys [<at 0 0 RCIM_Cubic>,<at 0.2 0.0134>,<at 0.4 0 RCIM_Cubic>]}
       RedCurve {keys [<at 0 0 RCIM_Cubic>,<at 0.2 0.1742>,<at 0.4 0 RCIM_Cubic>]}
      }})
    Create Color Attribute=true
    Create Exponent Attribute=true
    Create Radius Attribute=true
    Radius=100
   )
   /SetVariables(Particles.VisibilityTag=&self:PartHiddenBoneCheck(
     bisHandHidden:NiagaraBool=$User.bisHandHidden
     bisFootHidden:NiagaraBool=$User.bisFootHidden
     biaHeadHidden:NiagaraBool=$User.bisHeadHidden
     Selection:NiagaraInt32=$Particles.RandomBone
    )) {isSetParameters=true, module=SetVariables_#3}
  ]
  rendererClasses [&NiagaraSpriteRendererProperties, &NiagaraLightRendererProperties]
  renderers [
   {rendererClass=&NiagaraSpriteRendererProperties}
   {rendererClass=&NiagaraSpriteRendererProperties, rendererIndex=1}
   {rendererClass=&NiagaraLightRendererProperties, rendererIndex=2}
  ]
 }
 LargeArcs {
  emitterSpawnScript [
   &FN_PartitionAttributes(Particles Per Partition=10)
   /SetVariables(Emitter.Seed:NiagaraInt32=/Niagara/DynamicInputs/Random/RandomInt(Randomness Mode:ENiagaraRandomnessMode=::NewEnumerator0, Random Seed:NiagaraInt32=0, Fixed Random Seed:NiagaraBool=false)) {isSetParameters=true}
  ]
  emitterUpdateScript [
   &EmitterState(
    Life Cycle Mode=::NewEnumerator1
    Loop Duration=&UniformRangedFloat(Minimum=0.5, Maximum=2)
    Recalculate Duration Each Loop=true
    UseLoopDelay=false
   )
   &SpawnBurst_Instantaneous(Spawn Count=$Emitter.FN_PartitionAttributes.Spawn Total, Spawn Group=1, Use Spawn Probability=false)
   /SetVariables(
    Emitter.gen rand int:NiagaraInt32=&RandomRangeInt(Minimum=0, Maximum=3)
    Emitter.genRand:NiagaraFloat=&RandomRangeFloat(Minimum=0, Maximum=1, Evaluation Type=::NewEnumerator1)
   ) {isSetParameters=true}
  ]
  particleSpawnScript [
   &FN_Particle_Partition(Enable Ribbon ID=true, Seed:NiagaraInt32=$Emitter.Seed, Shuffle Partition ID=true) {enabled=false}
   /SetVariables(Particles.random:NiagaraFloat=$Emitter.genRand, Particles.randomInt:NiagaraInt32=$Emitter.gen rand int) {isSetParameters=true}
   &FN_Particle_Partition_v5(Enable Ribbon ID=true, Seed:NiagaraInt32=$Emitter.Seed, Shuffle Partition ID=true)
   /Niagara/Modules/Spawn/Initialization/InitializeParticle(
    Lifetime=0.5
    Module.Write Color=true
    Module.Write Lifetime=false
    Module.Write Mass=true
    Module.Write Material Random=true
    Module.Write Position=false
    Module.Write Scale=false
    Module.Write SpriteRotation=false
    Module.Write SpriteSize=true
    Position=&SimulationPosition()
    Sprite Size=<5 5>
   ) {enabled=false}
   /Niagara/Modules/Spawn/Initialization/V2/InitializeParticle(UsePositionOffset=false, Write Material Random=true) {enabled=false, module=InitializeParticle001}
   &InitializeRibbon(
    Color=<rgba 2.4952 0 6 1>
    Lifetime=&UniformRangedFloat(
     Minimum=0.15
     Maximum=0.3
     Randomness Mode=::NewEnumerator1
     Random Seed=&Add_Integer(A=$Particles.Partition.ID, B=$Emitter.LoopCount)
     Fixed Random Seed=true
    )
    Module.Write Color=true
    Module.Write Lifetime=true
    Module.Write Mass=true
    Module.Write Material Random=true
    Module.Write Position=false
    Module.Write RibbonTwist=false
    Module.Write RibbonWidth=true
    Ribbon Width=10
   )
   &FN_Initialize_BoneFlow_Attributes_v2(
    Angle:NiagaraFloat=&Multiply_Float(
     A=$Particles.Partition.NormalizedID
     B=&UniformRangedFloat(Minimum=0.75, Maximum=1, Random Seed=$Particles.Partition.ID)
    )
    Radius:NiagaraFloat=&UniformRangedFloat(
     Minimum=&FloatFromCurve(
      FloatCurve=DataInterface() {propertyValues {Curve {keys [
          <at 0 0 RCIM_Cubic>
          {arriveTangent=4.1777, interpMode=RCIM_Cubic, leaveTangent=4.1777, tangentMode=RCTM_User, time=0.2, value=20}
          {arriveTangent=-0.6996, interpMode=RCIM_Cubic, leaveTangent=-0.6996, tangentMode=RCTM_User, time=0.8, value=20}
          <at 1 0 RCIM_Cubic>
         ]}}}
      CurveIndex=$Particles.Partition.NormalizedID
     )
     Maximum=&FloatFromCurve(
      FloatCurve=DataInterface() {propertyValues {Curve {keys [
          <at 0 0 RCIM_Cubic>
          {arriveTangent=0.5753, interpMode=RCIM_Cubic, leaveTangent=0.5753, tangentMode=RCTM_User, time=0.2, value=5}
          {arriveTangent=0.7345, interpMode=RCIM_Cubic, leaveTangent=0.7345, tangentMode=RCTM_User, time=0.8, value=5}
          <at 1 0 RCIM_Cubic>
         ]}}}
      CurveIndex=$Particles.Partition.NormalizedID
     )
    )
    T:NiagaraFloat=$Particles.Partition.NormalizedID
   )
  ]
  particleUpdateScript [
   &ParticleState()
   &FN_BoneFlow(
    Angle:NiagaraFloat=&Add_Float(A=$Particles.Initial.FN_Initialize_BoneFlow_Attributes_v2.Angle, B=HlslExpression() {hlslExpression=Emitter.Age * .2})
    BoneID:NiagaraInt32=$Particles.randomInt
    Height:NiagaraFloat=$Particles.Partition.NormalizedID
    Interpolate Mesh Sampling=false
    Radius:NiagaraFloat=$Particles.Initial.FN_Initialize_BoneFlow_Attributes_v2.Radius
    Set Material Dynamic=false
    SkeletalMesh:NiagaraDataInterfaceSkeletalMesh=$User.SkeletalMesh
   )
   &ScaleRibbonWidth(Ribbon Width Scale=&Multiply_Float(
     A=&FloatFromCurve(FloatCurve=DataInterface() {propertyValues {Curve {keys [<at 0 1 RCIM_Cubic>,<at 0.2 3 RCIM_Cubic>,<at 0.5 1 RCIM_Cubic>]}}})
     B=&FloatFromCurve(
      FloatCurve=DataInterface() {propertyValues {Curve {keys [<at 0 0.5 RCIM_Cubic>,<at 0.4995 1.2 RCIM_Cubic>,<at 1 0.5 RCIM_Cubic>]}}}
      CurveIndex=$Particles.Partition.NormalizedID
     )
    ))
   &ScaleColor(
    Scale Alpha=&FloatFromCurve(FloatCurve=DataInterface() {propertyValues {Curve {keys [
         {arriveTangent=8.0553, interpMode=RCIM_Cubic, leaveTangent=8.0553, tangentMode=RCTM_User, time=0, value=0}
         <at 0.2113 1 RCIM_Cubic>
         <at 0.4546 0>
         {arriveTangent=-8.0018, interpMode=RCIM_Cubic, leaveTangent=7.1228, tangentMode=RCTM_Break, time=0.5517, value=0}
         <at 0.7524 1 RCIM_Cubic>
         {arriveTangent=-10.4882, interpMode=RCIM_Cubic, leaveTangent=-10.4882, tangentMode=RCTM_User, time=1, value=0}
        ]}}})
    ScaleA=true
    ScaleRGB=true
   )
   &RegionMask(
    Invert Region=true
    Offset:Vector3f=&MakeCustomVectorFromBool(Bool=$User.bisFootHidden, True Vector=<0 0 50>, False Vector=<0 0 0>)
    Offset Space:ENiagaraCoordinateSpace=::Local
    Region Type:ENiagaraDebugDrawShape=::NewEnumerator2
    Rotation:Vector3f=<0 0.25 0>
    UseOffset=true
    UseRotation=false
   )
   &ScaleColor(Scale Alpha=$Particles.RegionMask.Float, ScaleA=true, ScaleRGB=true) {module=ScaleColor001}
   &ScaleRibbonWidth(Ribbon Width Scale=&SelectFloatFromArray(
     Direct Array Index=&self:PartHiddenPartitionCheck(
      is Hand Hidden:NiagaraBool=$User.bisHandHidden
      Selection:NiagaraInt32=$Particles.Partition.ID
      is Foot Hidden:NiagaraBool=$User.bisFootHidden
     )
     Float Selection Array=DataInterface() {propertyValues {FloatData [1,0]}}
    )) {module=ScaleRibbonWidth001}
  ]
  rendererClasses [&NiagaraRibbonRendererProperties]
  renderers [{rendererClass=&NiagaraRibbonRendererProperties}]
 }
 Sparks {
  emitterUpdateScript [
   &EmitterState(Life Cycle Mode=::NewEnumerator1, Loop Behavior=::NewEnumerator1, Loop Duration=1, UseLoopDelay=false)
   &SpawnBurst_Instantaneous(Spawn Count=&UniformRangedInt(Minimum=10, Maximum=15), Use Spawn Probability=false)
  ]
  particleSpawnScript [
   /SetVariables(Particles.RandomBone:NiagaraInt32=&RandomRangeInt(Minimum=0, Maximum=21)) {isSetParameters=true}
   &InitializeParticle(
    Color=<rgba 1.767 0.6881 3 1>
    Lifetime Max=0.6
    Lifetime Min=0.4
    Lifetime Mode=::NewEnumerator1
    Sprite Size Max=<2 2>
    Sprite Size Min=<0.75 0.75>
    Sprite Size Mode=::NewEnumerator2
    UsePositionOffset=false
    Write Material Random=true
   ) {module=InitializeParticle001}
   &SkeletalMeshLocation(
    Bone / Socket Index=$Particles.RandomBone
    Bone Sampling Mode=::NewEnumerator4
    Inherit Velocity=false
    "Inherit Velocity (Scale)"=0
    MaskSpawnGroup=false
    Skeletal Mesh=$User.SkeletalMesh
   )
   &AddVelocityFromPoint(Module.Use Falloff Distance=false, Origin Offset=<0 0 100>, Velocity Strength=300)
   &SphereLocation(
    MaskSpawnGroup=false
    Override Seed=false
    Sphere Radius=10
    Surface Only=false
    UseNonUniformScale=false
    UseOffset=false
    UseSphereOrientation=false
   )
  ]
  particleUpdateScript [
   &ParticleState()
   &GravityForce(Gravity=<0 0 -400>)
   &Drag(Drag=6, Use Linear Drag=true, Use Rotational Drag=false)
   &SolveForcesAndVelocity(Clamp Velocity=false, Limit Acceleration=false)
   &ScaleSpriteSizeBySpeed(Max Scale Factor=<1 10>)
   &ScaleColor(
    Scale Alpha=&FloatFromCurve(FloatCurve=DataInterface() {propertyValues {Curve {keys [<at 0 1>,<at 1 0>]}}})
    Scale RGB=&VectorFromFloat(Value=&FloatFromCurve(FloatCurve=DataInterface() {propertyValues {Curve {keys [<at 0 1>,<at 1 0>]}}}))
    ScaleA=true
    ScaleRGB=true
   )
   &RegionMask(
    Invert Region=true
    Offset:Vector3f=&MakeCustomVectorFromBool(Bool=$User.bisFootHidden, True Vector=<0 0 50>, False Vector=<0 0 0>)
    Offset Space:ENiagaraCoordinateSpace=::Local
    Region Type:ENiagaraDebugDrawShape=::NewEnumerator2
    Rotation:Vector3f=<0 0.25 0>
    UseOffset=true
    UseRotation=false
   )
   &ScaleColor(Scale Alpha=$Particles.RegionMask.Float, ScaleA=true, ScaleRGB=true) {module=ScaleColor001}
   /SetVariables(Particles.VisibilityTag=&self:PartHiddenBoneCheck(
     bisHandHidden:NiagaraBool=$User.bisHandHidden
     bisFootHidden:NiagaraBool=$User.bisFootHidden
     biaHeadHidden:NiagaraBool=$User.bisHeadHidden
     Selection:NiagaraInt32=$Particles.RandomBone
    )) {isSetParameters=true}
  ]
  rendererClasses [&NiagaraSpriteRendererProperties]
  renderers [{rendererClass=&NiagaraSpriteRendererProperties}]
 }
 Spawn_Points {
  emitterUpdateScript [
   &EmitterState(Life Cycle Mode=::NewEnumerator1, Loop Duration=2, UseLoopDelay=false)
   &SpawnRate(
    SpawnRate=&FloatFromCurve(
     FloatCurve=DataInterface() {propertyValues {Curve {keys [<at 1.3 1>,<at 2.3 0>]}}}
     CurveIndex=&ModuloFloat(A=$Engine.Time, B=4)
     Scale Curve=8
    )
    Use Spawn Probability=false
   )
  ]
  particleSpawnScript [
   &InitializeParticle(Color Mode=::NewEnumerator0, Lifetime=0.0167, UsePositionOffset=false, Write Material Random=true)
   /SetVariables(Particles.New Vector:Vector3f=&RotateVectorByQuaternion(
     Vector To Rotate=Linked() {linkedVariable {name=Particles.Position, type=NiagaraPosition}}
     Quaternion=$Particles.SkeletalMeshLocation.SampledRotation
    )) {enabled=false, isSetParameters=true}
  ]
  particleUpdateScript [
   &ParticleState()
   &BoxLocation(
    Box Size=<22.5 22.5 22.5>
    MaskSpawnGroup=false
    Module.Override Seed=false
    Normalized Box Rotation Angle=0.125
    Rotate Box=true
    Rotate Box On Axis:Vector3f=<1 0 0>
    Surface Only=true
   )
   /SetVariables(Particles.SCaled Velocity:Vector3f=/Niagara/DynamicInputs/Vectors/RotateVector(
     Vector To Rotate:Vector3f=$Output.BoxLocation.BoxVector
     Yaw:NiagaraFloat=18
     Pitch:NiagaraFloat=0
     Roll:NiagaraFloat=0
     Rotation Rate:NiagaraFloat=1
    )) {isSetParameters=true}
   &SkeletalMeshLocation(
    Bone / Socket Index=21
    Bone Sampling Mode=::NewEnumerator4
    MaskSpawnGroup=false
    Orientation Sampling=::NewEnumerator1
    Sampled Position Offset=&Add_Vector(A=$Particles.SCaled Velocity, B=<6 2.5 0>)
    Skeletal Mesh=$User.SkeletalMesh
   )
   &ScaleColor(
    Scale Alpha=&FloatFromCurve(FloatCurve=DataInterface() {propertyValues {Curve {keys [<at -0.0046 1.0113 RCIM_Cubic>,<at 1.0011 0.002 RCIM_Cubic>]}}})
    ScaleA=true
    ScaleRGB=true
   )
   &SolveForcesAndVelocity(Clamp Velocity=false, Limit Acceleration=false)
   &GenerateDeathEvent(
    Event Probability:NiagaraFloat=&MakeCustomFloatFromBool(
     Bool=&SetBoolByFloatComparison(
      A=&DotProduct(
       Vector A=&NormalizeVector(
        Vector To Normalize=&Subtract_Vector(
         A=Linked() {linkedVariable {name=Particles.Position, type=NiagaraPosition}}
         B=Linked() {linkedVariable {name=Particles.SkeletalMeshLocation.SampledPosition, type=NiagaraPosition}}
        )
        Safe Normalize=true
        Safe Fallback Threshold=0
        Safe Fallback Vector=<0 0 0.001>
       )
       Vector B=&ReturnMeshOrientationAxis(Mesh Orientation Axis=::XAxis, Mesh Orientation Quat=$Particles.Initial.SkeletalMeshLocation.SampledRotation)
      )
      B=-1
     )
     True Float=1
     False Float=0
    )
    Use Event Probability=true
    "Vector to Send as Vector 2 (Velocity)":Vector3f=&RotateVector(
     Vector To Rotate:Vector3f=$Output.BoxLocation.BoxNormal
     Yaw:NiagaraFloat=-90
     Pitch:NiagaraFloat=45
     Roll:NiagaraFloat=120
     Rotation Rate:NiagaraFloat=1
    )
   ) {enabled=false}
   &GenerateDeathEvent(
    Event Probability=1
    Use Event Probability=true
    "Vector to Send as Vector 2 (Velocity)":Vector3f=&RotateVector(
     Vector To Rotate:Vector3f=$Output.BoxLocation.BoxNormal
     Yaw:NiagaraFloat=-90
     Pitch:NiagaraFloat=45
     Roll:NiagaraFloat=120
     Rotation Rate:NiagaraFloat=1
    )
   ) {module=GenerateDeathEvent001}
  ]
  rendererClasses [&NiagaraSpriteRendererProperties, &NiagaraComponentRendererProperties]
  renderers [
   {rendererClass=&NiagaraSpriteRendererProperties}
   {rendererClass=&NiagaraComponentRendererProperties, rendererIndex=1}
  ]
 }
}
summary {
 systemName=NS_KevinCoutureGold_Body_Loop
 emitters [
  {
   emitterName=BoneFlow_Burst
   rendererClasses [&NiagaraSpriteRendererProperties, &NiagaraLightRendererProperties]
  }
  {emitterName=FeetLights, rendererClasses [&NiagaraSpriteRendererProperties, &NiagaraLightRendererProperties]}
  {emitterName=BurstRunes, rendererClasses [&NiagaraSpriteRendererProperties]}
  {emitterName=AirGlow, rendererClasses [&NiagaraSpriteRendererProperties]}
  {emitterName=LargeArcs, rendererClasses [&NiagaraRibbonRendererProperties]}
  {
   emitterName=InitialBurst_Lightning
   rendererClasses [&NiagaraSpriteRendererProperties, &NiagaraLightRendererProperties]
  }
  {emitterName=Decal, rendererClasses [&NiagaraSpriteRendererProperties]}
  {emitterName=Sparks, rendererClasses [&NiagaraSpriteRendererProperties]}
  {emitterName=Bone_Flow_Ribbon02, rendererClasses [&NiagaraRibbonRendererProperties]}
  {
   emitterName=Spawn_Points
   rendererClasses [&NiagaraSpriteRendererProperties, &NiagaraComponentRendererProperties]
  }
  {emitterName=Gold_Flakes001, rendererClasses [&NiagaraSpriteRendererProperties, &NiagaraMeshRendererProperties]}
  {emitterName=Gold_Flakes002, rendererClasses [&NiagaraSpriteRendererProperties, &NiagaraMeshRendererProperties]}
  {emitterName=BoneFlow_Loop, rendererClasses [&NiagaraSpriteRendererProperties, &NiagaraLightRendererProperties]}
 ]
 userVariables [
  User.bisFootHidden {type=NiagaraBool}
  User.bisHandHidden {type=NiagaraBool}
  User.bisHeadHidden {type=NiagaraBool}
  User.SkeletalMesh {
   type=NiagaraDataInterfaceSkeletalMesh
   defaultValue {
    struct=NiagaraExt_VariableValue_DataInterface
    value {dataInterface=&self:NiagaraDataInterfaceSkeletalMesh_0}
   }
  }
  User.StatChangedKills {type=NiagaraFloat, defaultValue=1, description=""}
 ]
}
