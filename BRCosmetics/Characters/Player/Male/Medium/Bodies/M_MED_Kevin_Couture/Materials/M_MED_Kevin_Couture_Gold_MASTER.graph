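# Master character material for Kevin Couture Gold, layering an animated metallic gold effect over the standard character and skin FX attributes.
# A time-driven sine wave, measured by pre-skinned distance from a mask point and broken up by world-aligned cloud noise, creates gold and glowing transition-edge masks.
# Gold overrides color and surface response, adding parallax noise sparkles, masked runes and inner glow.
# The result passes through character and cinematic PostFX functions.
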Material /BRCosmetics/Characters/Player/Male/Medium/Bodies/M_MED_Kevin_Couture/Materials/M_MED_Kevin_Couture_Gold_MASTER
graph {
 properties {
  bAutomaticallySetUsageInEditor=false
  bUseMaterialAttributes=true
  bUsedWithMeshDeformer=true
  bUsedWithMorphTargets=true
  bUsedWithSkeletalMesh=true
  pixelDepthOffsetMode=PDOM_Legacy
  refractionMethod=RM_IndexOfRefraction
  shadingModel=MSM_FromMaterialExpression
  subsurfaceProfile=&SS_HeroSkin_02
  usedShadingModels=Default Lit | Subsurface | Clear Coat | Subsurface Profile | Cloth | Eye
 }
 nodes [
  n1 = MaterialFunctionCall() {materialFunction=&BaseFN_CharacterFunction_PreFX}
  n2 = GetMaterialAttributes(None=MaterialFunctionCall(
    PreskinnedPosition=@n1.Pre skinned Local Position
    Gmap=@n1.Pre skinned Local Position
    MaterialAttributes=@n1.Pre skinned Local Position
    AO=@n1.Pre skinned Local Position
   ) {materialFunction=&CharacterSkinFX}.Result) {attributeGetTypes [
    #1
    #2
    #3
    #4
    #5
    #6
   ]}
  n3 = VectorParameter() {group=Kevin Couture Gold, parameterName=Gold_BaseColor, defaultValue=<rgba 0.7292 0.2918 0 0>}
  n4 = VectorParameter() {group=Kevin Couture Gold, parameterName=Gold_BaseColor_2, defaultValue=<rgba 0.291 0.0983 0 0>}
  n5 = TextureSampleParameter2D() {
   group=Kevin Couture Gold
   parameterName=Emissive
   samplerSource=SSM_Wrap_WorldGroupSettings
   texture=&T_BlacksRGB
  }
  n6 = Fresnel() {baseReflectFraction=0, exponent=1}
  n7 = Power(Base=@n6) {constExponent=5}
  n8 = Saturate(None=StaticSwitchParameter(True=@n5.R, False=Multiply(A=@n7) {constB=2}) {group=Kevin Couture Gold, parameterName=Gold_UseEmissiveMaskBaseColorAdjustment})
  n9 = NamedRerouteUsage()
  n10 = ScalarParameter() {group=Kevin Couture Gold, parameterName=Gold_Specular_Min}
  n11 = ScalarParameter() {defaultValue=0.5, group=Kevin Couture Gold, parameterName=Gold_Specular_Max}
  n12 = NamedRerouteUsage()
  n13 = TextureSample(UVs=Multiply(A=TextureCoordinate()) {constB=8}) {samplerSource=SSM_Wrap_WorldGroupSettings, samplerType=SAMPLERTYPE_Masks, texture=/Game/Global/Textures/Default/Noise/T_CubeNoise}
  n14 = MaterialFunctionCall(In=@n13.R, Contrast=Constant() {r=-0.4}) {materialFunction=&CheapContrast}
  n15 = Multiply(A=ScalarParameter() {defaultValue=0.3, group=Kevin Couture Gold, parameterName=Gold_Roughness_A}, B=@n14.Result)
  n16 = Multiply(A=ScalarParameter() {defaultValue=0.6, group=Kevin Couture Gold, parameterName=Gold_Roughness_B}, B=@n14.Result)
  n17 = NamedRerouteUsage()
  n18 = Time()
  n19 = Add(A=ScalarParameter() {defaultValue=8, group=Kevin Couture Gold, parameterName=Gold_Animation_Duration, sortPriority=101}) {constB=0}
  n20 = Distance(
   A=VectorParameter() {group=Kevin Couture Gold, parameterName=Gold_MaskPos_01, defaultValue=<rgba -600 -300 500 1>}.RGB
   B=NamedRerouteUsage()
  )
  n21 = ScalarParameter() {defaultValue=155, group=Kevin Couture Gold, parameterName=Gold_Wavelength}
  n22 = MaterialFunctionCall(Value=Add(A=Divide(A=@n18, B=@n19), B=Divide(A=@n20, B=@n21))) {materialFunction=&LinearSine}
  n23 = Saturate(None=Add(A=@n22.Linear Sine) {constB=0.25})
  n24 = Subtract(A=@n23) {constB=0.5}
  n25 = Subtract(A=Multiply(A=Saturate(None=@n24)) {constB=2}, B=NamedRerouteUsage())
  n26 = Saturate(None=Subtract(A=@n23) {constB=0.275})
  n27 = Round(None=Subtract(A=Multiply(A=@n26) {constB=2}, B=NamedRerouteUsage()))
  n28 = Saturate(None=Multiply(A=Subtract(A=@n25, B=@n27)) {constB=20})
  n29 = MaterialFunctionCall(
   LDR Input=NamedRerouteDeclaration(None=Round(None=@n28)) {name=Gold_TransitionMask, nodeColor=<rgba 0 1 0.9531 1>}
   HDR Tint=VectorParameter() {group=Kevin Couture Gold, parameterName=Gold_EdgeGlowColor, defaultValue=<rgba 0.6444 0 1 1>}.RGB
   HDR Intensity Multiplier=ScalarParameter() {defaultValue=0.25, group=Kevin Couture Gold, parameterName=Gold_EdgeGlowStrength}
  ) {materialFunction=&DeriveHDRfromLDR}
  n30 = NamedRerouteUsage()
  n31 = Add(A=NamedRerouteUsage(), B=NamedRerouteUsage())
  n32 = SetMaterialAttributes!(
   MaterialAttributes=@n2.MaterialAttributes
   Base Color=LinearInterpolate(A=@n2.Base Color, B=LinearInterpolate(A=@n3.RGB, B=@n4.RGB, Alpha=@n8), Alpha=@n9)
   Specular=LinearInterpolate(A=@n2.Specular, B=LinearInterpolate(A=@n10, B=@n11, Alpha=@n12) {constB=0.5}, Alpha=@n9)
   Roughness=LinearInterpolate(A=@n2.Roughness, B=LinearInterpolate(A=@n15, B=@n16, Alpha=@n17) {constB=0.5}, Alpha=@n9)
   Emissive Color=LinearInterpolate(A=@n2.Emissive Color, B=Add(A=@n29.RemappedRange, B=@n30), Alpha=Saturate(None=@n31))
   Clear Coat or Cloth or Iris Mask or Curvature or Custom Data 0=LinearInterpolate(A=@n2.Clear Coat or Cloth or Iris Mask or Curvature or Custom Data 0, Alpha=@n9) {constB=0}
   Curvature or Metallic=LinearInterpolate(A=@n2.Curvature or Metallic, B=ScalarParameter() {defaultValue=0.9, group=Kevin Couture Gold, parameterName=Gold_Metalness}, Alpha=@n9) {constB=0.5}
  ) {attributeSetTypes [
    #1
    #2
    #3
    #4
    #5
    #6
   ]}
  n33 = MaterialFunctionCall(AO=@n1.AO, Distance Blend=@n1.AO, Fresnel=@n1.AO, Gmap=@n1.AO, Pre Skinned Local Position=@n1.AO, Material=@n32) {materialFunction=&BaseFN_CharacterFunction_PostFX}
  MaterialFunctionCall(In=@n33.Result) {materialFunction=&CINE_MF_CharPostFX}
  NamedRerouteDeclaration(None=VertexInterpolator(VS=PreSkinnedPosition()).PS) {name=Pre-SkinnedLocalPos_PS, nodeColor=<rgba 0 1 0.4609 1>}
  n34 = PreSkinnedNormal()
  n35 = ComponentMask(None=MaterialFunctionCall(
    TextureObject=TextureObject() {texture=&AtmosphericCloudNoise01}
    TextureSize=VectorParameter() {group=Kevin Couture Gold, parameterName=Gold_TransitionNoise_Tiling, defaultValue=<rgba 150 150 120 1>}.RGB
    WorldPosition=NamedRerouteUsage()
    World Space Normal=VertexInterpolator(VS=@n34).PS
   ) {materialFunction=/Game/Weapons/WeaponSkins/Wraps/MaterialFunctions/WorldAlignedTexture}.XYZ Texture) {r=true}
  n36 = Power(Base=Add(A=@n35) {constB=0.2}) {constExponent=0.2}
  NamedRerouteDeclaration(None=LinearInterpolate(Alpha=Saturate(None=@n36)) {constB=0.3}) {name=Gold_LowContrastBreakup, nodeColor=<rgba 1 0 0.7031 1>}
  n37 = Saturate(None=Subtract(A=@n25) {constB=0.05})
  NamedRerouteDeclaration(None=Saturate(None=Multiply(A=@n37) {constB=40})) {name=Gold_Mask, nodeColor=<rgba 0.9453 1 0 1>}
  n38 = TextureSampleParameter2D() {
   parameterName=SkinFX_Mask
   samplerSource=SSM_Wrap_WorldGroupSettings
   samplerType=SAMPLERTYPE_Masks
   texture=&T_BlackMask
  }
  n39 = Constant() {r=0.5}
  n40 = TextureCoordinate()
  n41 = Multiply(A=Constant() {r=10}, B=@n40)
  n42 = Constant() {r=-2}
  n43 = Constant() {r=0.02}
  n44 = TextureSample(UVs=Multiply(A=@n39, B=BumpOffset(Coordinate=@n41, Height=@n42, HeightRatioInput=@n43))) {samplerSource=SSM_Wrap_WorldGroupSettings, samplerType=SAMPLERTYPE_Masks, texture=&T_CubeNoise}
  n45 = Multiply(A=@n40, B=Constant() {r=20})
  n46 = Constant() {r=0.2}
  n47 = TextureSample(UVs=Multiply(A=@n39, B=BumpOffset(Coordinate=@n45, Height=@n42, HeightRatioInput=@n46))) {samplerSource=SSM_Wrap_WorldGroupSettings, samplerType=SAMPLERTYPE_Masks, texture=&T_CubeNoise}
  n48 = Multiply(A=Power(Base=Multiply(A=@n44.B, B=@n47.B), Exp=Constant() {r=7}), B=Constant() {r=60})
  n49 = Constant() {r=8}
  n50 = TextureCoordinate()
  n51 = TextureSample!(
   UVs=BumpOffset(Coordinate=Multiply(A=@n49, B=@n50), Height=Constant() {r=0.58}, HeightRatioInput=Constant() {r=-0.5})
   Tex
   Bias=ScalarParameter() {defaultValue=5, parameterName=mipbias1}
   Apply View MipBias
  ) {
   mipValueMode=TMVM_MipBias
   samplerSource=SSM_Wrap_WorldGroupSettings
   samplerType=SAMPLERTYPE_Masks
   texture=&T_CubeNoise
  }
  n52 = Add(A=Power(Base=@n51.R, Exp=Constant() {r=3}), B=Constant() {r=-0.05})
  n53 = Saturate(None=Multiply(A=NamedRerouteDeclaration(None=@n48) {name=Sparkles, nodeColor=<rgba 0 0.5078 1 1>}, B=Saturate(None=@n52)))
  n54 = Multiply(
   A=ScalarParameter() {defaultValue=50, parameterName=Gold_Rune_Intensity}
   B=Multiply(A=Add(A=@n38.A, B=@n53), B=VectorParameter() {group=Kevin Couture Gold, parameterName=Gold_RuneColor, defaultValue=<rgba 1 0.6535 0 0>}.RGB)
  )
  n55 = TextureSampleParameter2D() {
   group=Kevin Couture Gold
   parameterName=Emissive
   samplerSource=SSM_Wrap_WorldGroupSettings
   texture=/BRCosmetics/Characters/Player/Female/Medium/Bodies/F_MED_Punk_Koi_White/Textures/T_F_Med_Punk_Koi_White_Body_E
  }
  n56 = Multiply(
   A=Power(Base=OneMinus(None=@n6)) {constExponent=4}
   B=NamedRerouteDeclaration(None=@n13.G) {name=Gold_CubeNoise, nodeColor=<rgba 1 0.9609 0 1>}
  )
  n57 = StaticSwitchParameter(True=Multiply(A=TextureCoordinate()) {constB=4}, False=Multiply(A=TextureCoordinate()) {constB=2}) {group=Kevin Couture Gold, parameterName=Gold_UseEmissiveMaskBaseColorAdjustment}
  n58 = TextureSample!(UVs=@n57, Tex, Apply View MipBias) {samplerSource=SSM_Wrap_WorldGroupSettings, texture=&AtmosphericCloudNoise01}
  n59 = TextureSample(UVs=MaterialFunctionCall!(HeightTexture=@n58.G, TextureCoordinates=@n57, ReferencePlane=Constant() {r=0.5}, HeightRatioInput=Constant() {r=0.5}, CameraVector) {materialFunction=/Engine/Functions/Engine_MaterialFunctions02/Utility/BumpOffset_Advanced.BumpOffset_advanced}.UVs) {samplerSource=SSM_Wrap_WorldGroupSettings, texture=&AtmosphericCloudNoise01}
  n60 = Saturate(None=Power(Base=LinearInterpolate(A=@n59.G, B=@n58.G, Alpha=@n58.G)) {constExponent=0.5})
  n61 = Constant() {r=-0.45}
  n62 = Multiply(
   A=Multiply(
    A=VectorParameter() {group=Kevin Couture Gold, parameterName=Gold_InnerGlowColor, defaultValue=<rgba 1 0.3784 0 0>}.RGB
    B=ScalarParameter() {defaultValue=5, group=Kevin Couture Gold, parameterName=Gold_InteriorGlow}
   )
   B=Multiply(
    A=StaticSwitchParameter(True=@n55.R, False=@n56) {group=Kevin Couture Gold, parameterName=Gold_UseEmissiveMaskBaseColorAdjustment}
    B=MaterialFunctionCall(In=@n60, Contrast=@n61) {materialFunction=&CheapContrast}.Result
   )
  )
  NamedRerouteDeclaration(None=Multiply(A=NamedRerouteUsage(), B=Add(A=@n54, B=@n62))) {name=Gold_SparklesAndRunes, nodeColor=<rgba 0.9453 1 0 1>}
  n63 = Subtract(A=@n22.Linear Sine) {constB=0.3}
  n64 = NamedRerouteUsage()
  NamedRerouteDeclaration(None=Saturate(None=Subtract(A=@n63, B=@n64))) {name=Gold_EdgeMask, nodeColor=<rgba 1 0 0.7031 1>}
  NamedRerouteDeclaration(None=LinearInterpolate(
    A=ScalarParameter() {defaultValue=0.1, group=Kevin Couture Gold, parameterName=Gold_Transition_Falloff_Start}
    B=ScalarParameter() {defaultValue=0.6, group=Kevin Couture Gold, parameterName=Gold_Transition_Falloff_End}
    Alpha=@n35
   )) {name=Gold_RampBreakup, nodeColor=<rgba 1 0.9609 0 1>}
  NamedRerouteDeclaration(None=@n38.R) {name=Gold_OutfitMask, nodeColor=<rgba 1 0 0.4922 1>}
 ]
}
