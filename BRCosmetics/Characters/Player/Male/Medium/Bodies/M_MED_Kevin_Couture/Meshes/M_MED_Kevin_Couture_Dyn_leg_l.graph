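# Physics asset for the left-leg dynamic boot cuff of the Kevin Couture medium male player character body.
# Two capsule bodies cover a tiny calf anchor and a larger capsule on the boot cuff bone, offset along the bone.
# A single constraint ties the cuff to the calf with narrow swing limits and a tighter twist limit, keeping the cuff's secondary motion close to the leg.
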PhysicsAsset /BRCosmetics/Characters/Player/Male/Medium/Bodies/M_MED_Kevin_Couture/Meshes/M_MED_Kevin_Couture_Dyn_leg_l
bodies {
 calf_l [{shapeType=Capsule, length=0.5, radius=0.5}]
 dyn_boot_cuff_l [{shapeType=Capsule, length=5.875, radius=3.0312, center=<4 0.0001 0>, rotation=<rot 90 0 0>}]
}
constraints [{bone1Name=dyn_boot_cuff_l, bone2Name=calf_l, swing1LimitDegrees=15, swing2LimitDegrees=15, twistLimitDegrees=5}]
