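# Physics asset for a dangling chest string on the Kevin Couture male medium character's shirt.
# Three thin capsules form a chain from a small master body attached to spine_05, with moderate swing and tight twist limits between links.
# Larger capsules on spine_04 and spine_05, plus an angled box on spine_04, approximate the torso.
# The master-to-spine and spine-to-spine constraints leave swing and twist free.
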PhysicsAsset /BRCosmetics/Characters/Player/Male/Medium/Bodies/M_MED_Kevin_Couture/Meshes/M_MED_Kevin_Couture_Dyn_shirt_string
bodies {
 dyn_chest_string_01 [{shapeType=Capsule, length=4.6875, radius=0.5, center=<2.1855 0 0>, rotation=<rot 90 0 0>}]
 dyn_chest_string_01_MSTR [{shapeType=Capsule, length=0.505, radius=0.505}]
 dyn_chest_string_02 [{shapeType=Capsule, length=4.6875, radius=0.5, center=<2.1855 0 0>, rotation=<rot 90 0 0>}]
 dyn_chest_string_03 [{shapeType=Capsule, length=4.6875, radius=0.5, center=<2.1855 0 0>, rotation=<rot 90 0 0>}]
 spine_04 [
  {shapeType=Capsule, length=15.4375, radius=12.9062, center=<3.5104 0.107 0>}
  {shapeType=Box, extentX=14.0625, extentY=13.875, extentZ=13.875, center=<6.108 9.8292 -4.5647>, rotation=<rot -1.4723 -21.2985 13.392>}
 ]
 spine_05 [{shapeType=Capsule, length=18.0938, radius=10.25, center=<-2.0112 0.1862 0>}]
}
constraints [
 {bone1Name=dyn_chest_string_01, bone2Name=dyn_chest_string_01_MSTR, swing1LimitDegrees=45, swing2LimitDegrees=45, twistLimitDegrees=5}
 {bone1Name=dyn_chest_string_02, bone2Name=dyn_chest_string_01, swing1LimitDegrees=45, swing2LimitDegrees=45, twistLimitDegrees=5}
 {bone1Name=dyn_chest_string_03, bone2Name=dyn_chest_string_02, swing1LimitDegrees=45, swing2LimitDegrees=45, twistLimitDegrees=5}
 {
  bone1Name=dyn_chest_string_01_MSTR
  bone2Name=spine_05
  swing1LimitDegrees=45
  swing1Motion=Free
  swing2LimitDegrees=45
  swing2Motion=Free
  twistLimitDegrees=45
  twistMotion=Free
 }
 {
  bone1Name=spine_05
  bone2Name=spine_04
  swing1LimitDegrees=45
  swing1Motion=Free
  swing2LimitDegrees=45
  swing2Motion=Free
  twistLimitDegrees=45
  twistMotion=Free
 }
]
