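# Physics asset for the left boot strings on the Kevin Couture medium male player cosmetic, apparently for dynamic secondary motion.
# Small capsules cover the inner and outer loops and ends of the lace around a master string bone, alongside a larger capsule on the left foot.
# Each loop and end is constrained directly to the master with moderate swing and tight twist limits, while the foot capsule has no constraint.
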PhysicsAsset /BRCosmetics/Characters/Player/Male/Medium/Bodies/M_MED_Kevin_Couture/Meshes/M_MED_Kevin_Couture_Dyn_strings_l
bodies {
 dyn_boot_string_end_in_l [{shapeType=Capsule, length=3, radius=0.5, center=<5.1974 0.402 1.1398>, rotation=<rot 69.2345 -172.5486 -177.7775>}]
 dyn_boot_string_end_out_l [{shapeType=Capsule, length=2.875, radius=0.5, center=<5.1809 -0.511 0.6149>, rotation=<rot 79.6928 -170.2956 -160.5963>}]
 dyn_boot_string_l_MSTR [{shapeType=Capsule, length=0.5, radius=0.5}]
 dyn_boot_string_loop_out_l [{shapeType=Capsule, length=3.375, radius=0.8438, center=<3.6602 -0.9726 1.1859>, rotation=<rot 47.0852 -118.1135 -111.369>}]
 dyn_boot_string_roop_in_l [{shapeType=Capsule, length=3.0938, radius=1.125, center=<3.4771 0.0802 0.5527>, rotation=<rot -2.5968 -88.5646 -97.6585>}]
 foot_l [{shapeType=Capsule, length=12.5435, radius=6.1379, center=<-4.933 6.1805 0.0001>, rotation=<rot 90 0 60.0009>}]
}
constraints [
 {bone1Name=dyn_boot_string_end_out_l, bone2Name=dyn_boot_string_l_MSTR, swing1LimitDegrees=30, swing2LimitDegrees=30, twistLimitDegrees=5}
 {bone1Name=dyn_boot_string_loop_out_l, bone2Name=dyn_boot_string_l_MSTR, swing1LimitDegrees=30, swing2LimitDegrees=30, twistLimitDegrees=5}
 {bone1Name=dyn_boot_string_end_in_l, bone2Name=dyn_boot_string_l_MSTR, swing1LimitDegrees=30, swing2LimitDegrees=30, twistLimitDegrees=5}
 {bone1Name=dyn_boot_string_roop_in_l, bone2Name=dyn_boot_string_l_MSTR, swing1LimitDegrees=30, swing2LimitDegrees=30, twistLimitDegrees=5}
]
