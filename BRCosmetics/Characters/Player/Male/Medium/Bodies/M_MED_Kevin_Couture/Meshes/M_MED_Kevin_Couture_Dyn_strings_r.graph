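# Physics asset for the dynamic right boot strings on the Kevin Couture medium male player body.
# Small capsules cover the inner and outer lace loops and ends, each constrained directly to the master string bone with limited swing and tight twist, forming a star rather than chains.
# A large capsule on the right foot has no constraint, apparently serving as a collision volume for the strings.
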PhysicsAsset /BRCosmetics/Characters/Player/Male/Medium/Bodies/M_MED_Kevin_Couture/Meshes/M_MED_Kevin_Couture_Dyn_strings_r
bodies {
 dyn_boot_string_end_in_r [{shapeType=Capsule, length=3, radius=0.5, center=<5.3135 -0.4112 0.9208>, rotation=<rot 60.9258 -175.8458 -171.9819>}]
 dyn_boot_string_end_out_r [{shapeType=Capsule, length=2.875, radius=0.5, center=<5.1894 0.5382 0.3556>, rotation=<rot 73.7349 -170.9142 -178.9243>}]
 dyn_boot_string_r_MSTR [{shapeType=Capsule, length=0.5, radius=0.5}]
 dyn_boot_string_roop_in_r [{shapeType=Capsule, length=3.0938, radius=1.125, center=<3.8346 0.0886 0.5628>, rotation=<rot -2.5968 -88.5646 -97.6585>}]
 dyn_boot_string_roop_out_r [{shapeType=Capsule, length=3.375, radius=0.8438, center=<3.8113 1.0065 0.9>, rotation=<rot 41.9086 -78.4183 -101.6991>}]
 foot_r [{shapeType=Capsule, length=12.5435, radius=6.1379, center=<4.933 -6.1805 -0.0001>, rotation=<rot 90 0 60.0009>}]
}
constraints [
 {bone1Name=dyn_boot_string_roop_in_r, bone2Name=dyn_boot_string_r_MSTR, swing1LimitDegrees=30, swing2LimitDegrees=30, twistLimitDegrees=5}
 {bone1Name=dyn_boot_string_end_in_r, bone2Name=dyn_boot_string_r_MSTR, swing1LimitDegrees=30, swing2LimitDegrees=30, twistLimitDegrees=5}
 {bone1Name=dyn_boot_string_end_out_r, bone2Name=dyn_boot_string_r_MSTR, swing1LimitDegrees=30, swing2LimitDegrees=30, twistLimitDegrees=5}
 {bone1Name=dyn_boot_string_roop_out_r, bone2Name=dyn_boot_string_r_MSTR, swing1LimitDegrees=30, swing2LimitDegrees=30, twistLimitDegrees=5}
]
