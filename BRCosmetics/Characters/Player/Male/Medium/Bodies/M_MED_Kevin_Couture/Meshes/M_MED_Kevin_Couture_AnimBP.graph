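# Animation Blueprint for the Kevin Couture player body cosmetic, adding secondary motion to its legs, boot strings and shirt string.
# It copies the attached parent mesh's pose, then uses weighted multi-bone constraints to position the chest and boot string master bones.
# RigidBody nodes simulate both legs, both string sets and the shirt string from separate physics assets, in calf or spine bone space.
# The event graph has no logic.
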AnimBlueprint /BRCosmetics/Characters/Player/Male/Medium/Bodies/M_MED_Kevin_Couture/Meshes/M_MED_Kevin_Couture_AnimBP
target_skeleton=&/M_MED_Kevin_Couture_Skeleton
graphs [
 AnimGraph {nodes [
   n1 = "Constraint - dyn_chest_string_01_MSTR (PPPPPPP)"!(
    ComponentPose=Local To Component(LocalPose=Copy Pose From Mesh() {bUseAttachedParent=true})
    Alpha=1
    ConstraintWeights_0=0.0106
    ConstraintWeights_1:float=0.0705
    ConstraintWeights_2:float=0.1407
    ConstraintWeights_3:float=0.6515
    ConstraintWeights_4:float=0.057
    ConstraintWeights_5:float=0.0491
    ConstraintWeights_6:float=0.0205
   ) -> (Pose) {
    alphaBoolBlend {blendOption=Linear}
    boneToModify {boneName=dyn_chest_string_01_MSTR}
    constraintSetup [
     {targetBone {boneName=upperarm_twist_01_r}}
     {targetBone {boneName=spine_03}}
     {targetBone {boneName=spine_04}}
     {targetBone {boneName=spine_05}}
     {targetBone {boneName=clavicle_r}}
     {targetBone {boneName=neck_02}}
     {targetBone {boneName=neck_01}}
    ]
    constraintWeights [1,1,1,1,1,1,1]
   }
   n2 = RigidBody!(ComponentPose=@n1, Alpha, OverrideWorldGravity:Vector="0,0,-980", ExternalForce) -> (Pose) {
    bClampLinearTranslationLimitToRefPose=true
    bOverrideWorldGravity=true
    overridePhysicsAsset=&/M_MED_Kevin_Couture_Dyn_leg_l
    simulationSpace=BaseBoneSpace
    alphaBoolBlend {blendOption=Linear}
    baseBoneRef {boneName=calf_l}
    componentAppliedLinearAccClamp=<2000 3000 3000>
    componentLinearAccScale=<0.1 0.35 0.35>
    componentLinearVelScale=<0.1 0.35 0.35>
    overrideWorldGravity {z=-980}
    simSpaceSettings {maxAngularAcceleration=500, maxAngularVelocity=500, velocityScaleZ=0, worldAlpha=0.05, externalLinearDragV=<0.1 0.1 0.1>}
   }
   n3 = RigidBody(ComponentPose=@n2, OverrideWorldGravity="0,0,-980") {
    bClampLinearTranslationLimitToRefPose=true
    bOverrideWorldGravity=true
    overridePhysicsAsset=&/M_MED_Kevin_Couture_Dyn_leg_r
    simulationSpace=BaseBoneSpace
    alphaBoolBlend {blendOption=Linear}
    baseBoneRef {boneName=calf_r}
    componentAppliedLinearAccClamp=<2000 3000 3000>
    componentLinearAccScale=<0.1 0.35 0.35>
    componentLinearVelScale=<0.1 0.35 0.35>
    overrideWorldGravity {z=-980}
    simSpaceSettings {maxAngularAcceleration=500, maxAngularVelocity=500, velocityScaleZ=0, worldAlpha=0.05, externalLinearDragV=<0.1 0.1 0.1>}
   }
   n4 = "Constraint - dyn_boot_string_l_MSTR (PPP)"!(ComponentPose=@n3, Alpha=1, ConstraintWeights_0=0.9379, ConstraintWeights_1:float=0.0044, ConstraintWeights_2:float=0.0577) -> (Pose) {
    alphaBoolBlend {blendOption=Linear}
    boneToModify {boneName=dyn_boot_string_l_MSTR}
    constraintSetup [{targetBone {boneName=dyn_boot_cuff_l}}, {targetBone {boneName=calf_twist_01_l}}, {targetBone {boneName=foot_l}}]
    constraintWeights [1,1,1]
   }
   n5 = "Constraint - dyn_boot_string_r_MSTR (PPP)"!(ComponentPose=@n4, Alpha=1, ConstraintWeights_0=0.0044, ConstraintWeights_1:float=0.0577, ConstraintWeights_2:float=0.9379) -> (Pose) {
    alphaBoolBlend {blendOption=Linear}
    boneToModify {boneName=dyn_boot_string_r_MSTR}
    constraintSetup [{targetBone {boneName=calf_twist_01_r}}, {targetBone {boneName=foot_r}}, {targetBone {boneName=dyn_boot_cuff_r}}]
    constraintWeights [1,1,1]
   }
   n6 = RigidBody(ComponentPose=@n5, OverrideWorldGravity="0,0,-980") {
    bClampLinearTranslationLimitToRefPose=true
    bOverrideWorldGravity=true
    overridePhysicsAsset=&/M_MED_Kevin_Couture_Dyn_strings_l
    simulationSpace=BaseBoneSpace
    alphaBoolBlend {blendOption=Linear}
    baseBoneRef {boneName=calf_l}
    componentAppliedLinearAccClamp=<2000 3000 3000>
    componentLinearAccScale=<0.1 0.35 0.35>
    componentLinearVelScale=<0.1 0.35 0.35>
    overrideWorldGravity {z=-980}
    simSpaceSettings {maxAngularAcceleration=500, maxAngularVelocity=500, velocityScaleZ=0, worldAlpha=0.05, externalLinearDragV=<0.1 0.1 0.1>}
   }
   n7 = RigidBody(ComponentPose=@n6, OverrideWorldGravity="0,0,-980") {
    bClampLinearTranslationLimitToRefPose=true
    bOverrideWorldGravity=true
    overridePhysicsAsset=&/M_MED_Kevin_Couture_Dyn_strings_r
    simulationSpace=BaseBoneSpace
    alphaBoolBlend {blendOption=Linear}
    baseBoneRef {boneName=calf_twist_01_r}
    componentAppliedLinearAccClamp=<2000 3000 3000>
    componentLinearAccScale=<0.1 0.35 0.35>
    componentLinearVelScale=<0.1 0.35 0.35>
    overrideWorldGravity {z=-980}
    simSpaceSettings {maxAngularAcceleration=500, maxAngularVelocity=500, velocityScaleZ=0, worldAlpha=0.05, externalLinearDragV=<0.1 0.1 0.1>}
   }
   n8 = RigidBody(ComponentPose=@n7, OverrideWorldGravity="0,0,-980") {
    bClampLinearTranslationLimitToRefPose=true
    bOverrideWorldGravity=true
    overridePhysicsAsset=&/M_MED_Kevin_Couture_Dyn_shirt_string
    simulationSpace=BaseBoneSpace
    alphaBoolBlend {blendOption=Linear}
    baseBoneRef {boneName=spine_05}
    componentAppliedLinearAccClamp=<2000 3000 3000>
    componentLinearAccScale=<0.1 0.35 0.35>
    componentLinearVelScale=<0.1 0.35 0.35>
    overrideWorldGravity {z=-980}
    simSpaceSettings {maxAngularAcceleration=500, maxAngularVelocity=500, velocityScaleZ=0, worldAlpha=0.05, externalLinearDragV=<0.1 0.1 0.1>}
   }
   Output Pose(Result=Component To Local(ComponentPose=@n8)) {layerGroup=None}
  ]}
 EventGraph {nodes [Event BlueprintUpdateAnimation()]}
]
variables []
